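# UI material for the Stage Show event editor tool that previews a beat-driven signal curve across one widget.
# MF_SparksStageShow_SignalProcessor evaluates the signal along U from the SignalScaleOffset, SignalProfile and AnimationMode parameters.
# Subtracting inverted V and dividing by the screen-space derivative length yields an antialiased fill beneath the curve.
# That yellow fill fades vertically over a dark background of four repeating gradient bands.
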Material /SparksStageShow/Systems/StageManager_v2_Prototype/Tools/EventEditor/Materials/M_StageShowUtility_SignalPreview
graph {
 properties {bAutomaticallySetUsageInEditor=false, materialDomain=MD_UI, usedShadingModels=Default Lit}
 nodes [
  n1 = MaterialFunctionCall(Float2=TextureCoordinate()) {materialFunction=&BreakOutFloat2Components}
  n2 = OneMinus(None=@n1.R)
  n3 = Add(A=Frac(None=Multiply(A=@n2) {constB=4})) {constB=0.8}
  n4 = Constant4Vector() {constant=<rgba 1 0.9684 0.3565 1>}
  n5 = VectorParameter() {parameterName=SignalScaleOffset, channelNames {a=Min Value, b=Peak Value, g=Beat Offset, r=Beat Multiplier}, defaultValue=<rgba 1 0 1 0>}
  n6 = VectorParameter() {parameterName=SignalProfile, channelNames {a=Fall Easing, b=Fall Time, g=Rise Easing, r=Rise Time}, defaultValue=<rgba 2 1 2 1>}
  n7 = OneMinus(None=@n1.G)
  n8 = Subtract(
   A=MaterialFunctionCall!(
    Measure Signal=@n1.R
    Beats Per Measure
    Beat Multiplier=@n5.Beat Multiplier
    BeatOffset=@n5.Beat Multiplier
    Peak Value=@n5.Beat Multiplier
    Min Value=@n5.Beat Multiplier
    Rise Time=@n6.Rise Time
    Rise Easing=@n6.Rise Time
    Fall Time=@n6.Rise Time
    Fall Easing=@n6.Rise Time
    AnimationMode=ScalarParameter() {defaultValue=1, parameterName=AnimationMode}
   ) {materialFunction=/SparksStageShow/Systems/StageManager_v2_Prototype/MaterialFunctions/MF_SparksStageShow_SignalProcessor}.Signal
   B=@n7
  )
  n9 = MaterialFunctionCall(X=DDX(Value=@n8), Y=DDY(Value=@n8)) {materialFunction=&MakeFloat2}
  n10 = Saturate(None=Divide(A=@n8, B=Length(None=@n9.Result) {is=class MaterialExpressionLength}))
  LinearInterpolate(
   A=Multiply(A=Multiply(A=@n3) {constB=0.5}, B=Constant4Vector() {constant=<rgba 0.0208 0.0208 0.0208 1>}.RGBA)
   B=Multiply(A=Multiply(A=@n4.RGBA, B=@n10))
   Alpha=Multiply(A=Multiply(A=@n10, B=@n7)) {constB=0.8}
  )
 ]
}
