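# Editor utility widget for the prototype stage show event editor that edits a signal phase change's duration and easing.
# The beat-labelled radial slider stores Duration as the reciprocal of its value, while its tooltip shows the raw value.
# The Out/In slider, optionally inverted, is remapped piecewise around its midpoint into an exponential Easing factor.
# Both sliders broadcast OnValueChanged with both parameters, and SetInteractivity enables or disables them.
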EditorUtilityWidgetBlueprint /SparksStageShow/Systems/StageManager_v2_Prototype/Tools/EventEditor/Widgets/Sparks_StageShowUtility_SignalPhaseParams
graphs [
 EventGraph {nodes [
   n1 = "On Value Changed (DurationRadialSlider)"!() -> (OutputDelegate, then, Value:float) {is=title K2Node_ComponentBoundEvent}
   > Set Duration(Duration:double=float / float(A=1, B=@n1.Value)) -> (Output_Get:double)
   > SetToolTipText!(
    execute
    self:Widget Object Reference=Get DurationRadialSlider() -> (DurationRadialSlider:Radial Slider Object Reference)
    InToolTipText:"Text (by ref)"="To Text (Float)"(Value=@n1.Value, MinimumIntegralDigits=1, MaximumIntegralDigits=324, MinimumFractionalDigits=0, MaximumFractionalDigits=3)
   ) -> (then)
   > Call On Value Changed!(
    execute:Exec
    self:Sparks Stage Show Utility Signal Phase Params Object Reference
    Duration:double=Get Duration() -> (Duration:double)
    Easing:double=Get Easing() -> (Easing:double)
   ) -> (then:Exec) {is=title K2Node_CallDelegate}
   n2 = "On Value Changed (EasingSlider)"!() -> (OutputDelegate, then, Value:float) {is=title K2Node_ComponentBoundEvent}
   n3 = SelectFloat(A=float - float(A=1, B=@n2.Value), B=@n2.Value, bPickA=Get InvertEasing() -> (InvertEasing:bool))
   > n4 = Branch(Condition="float >= float"(A=@n3, B=0.5))
   > n5 = Set Easing(Easing:double=Remap Float(Input=@n3, OldRange=(X=0.5,Y=1), NewRange=(X=1,Y=4), Clamp=true) -> (Output=~)) -> (Output_Get:double)
   n6 = Get EasingSlider() -> (EasingSlider:Editor Utility Slider Object Reference)
   n7 = Get Value!(self:Slider Object Reference=@n6) -> (Value:float)
   SetToolTipText(
    execute=[@n8.then, @n5.then]
    self=@n6
    InToolTipText="To Text (Float)"(
     Value=Remap Float(Input=@n7, OldRange=(X=0,Y=1), NewRange=(X=-1,Y=1), Clamp=true) -> (Output=~)
     MinimumIntegralDigits=1
     MaximumIntegralDigits=324
     MinimumFractionalDigits=0
     MaximumFractionalDigits=1
    )
   )
   > Call On Value Changed(Duration=Get Duration(), Easing=Get Easing())
   n8 = Set Easing(execute=@n4.else, Easing=Remap Float(Input=@n3, OldRange=(X=0.5,Y=0), NewRange=(X=0.99,Y=0.25), Clamp=true) -> (Output=~))
   n9 = SetInteractivity!() -> (OutputDelegate, then, Enabled:bool=~) {is=title K2Node_CustomEvent}
   > SetIsEnabled(self=Get DurationRadialSlider(), bInIsEnabled=@n9.Enabled)
   > SetIsEnabled(self=Get EasingSlider(), bInIsEnabled=@n9.Enabled)
   Event PreConstruct()
   Event Construct()
   Event Tick()
  ]}
 OnValueChanged {nodes [OnValueChanged!() -> (then, Duration:double=~, Easing:double=~) {is=title K2Node_FunctionEntry}]}
]
variables [
 Duration {type=number, default=2}
 Easing {type=number, default=1}
 InvertEasing {type=boolean}
 OnValueChanged {default=()}
]
widgets {
 info {rootWidgetClass=&SizeBox, widgetCount=12}
 widgets [
  SizeBox_161 {
   widgetClassPath=&SizeBox
   widget=SizeBox_161
   properties {
    bOverride_MaxDesiredWidth=true
    bOverride_MinDesiredWidth=true
    heightOverride=150
    maxDesiredHeight=100
    maxDesiredWidth=160
    minDesiredHeight=150
    minDesiredWidth=100
    widthOverride=150
   }
  }
  VerticalBox_237 {
   widgetClassPath=&VerticalBox
   parent=SizeBox_161
   widget=VerticalBox_237
   slot {class=SizeBoxSlot, properties {padding=<ltrb 10 0 10 0>}}
  }
  Overlay_145 {
   widgetClassPath=&Overlay
   parent=VerticalBox_237
   widget=Overlay_145
   slot {class=VerticalBoxSlot, properties {size {sizeRule=Fill, value=1}}}
  }
  DurationRadialSlider {
   widgetClassPath=/Script/AdvancedWidgets.RadialSlider
   bIsVariable=true
   parent=Overlay_145
   widget=DurationRadialSlider
   properties {
    bUseCustomDefaultValue=true
    customDefaultValue=0.0625
    mouseUsesStep=true
    showSliderHand=true
    stepSize=0.0625
    toolTipText=0.25
    value=0.25
    centerBackgroundColor=<rgba 0 0 0 0.2>
    handStartEndRatio=<0.6 1>
    sliderProgressColor=<rgba 1 0.8963 0.0782 1>
    sliderRange {
     externalCurve=null
     editorCurveData {
      defaultValue=3.4028234663852886e+38
      postInfinityExtrap=RCCE_Constant
      preInfinityExtrap=RCCE_Constant
      keys [
       {
        arriveTangent=0
        arriveTangentWeight=0
        interpMode=RCIM_Linear
        leaveTangent=0
        leaveTangentWeight=0
        tangentMode=RCTM_Auto
        tangentWeightMode=RCTWM_WeightedNone
        time=0
        value=0.0625
       }
       {
        arriveTangent=0
        arriveTangentWeight=0
        interpMode=RCIM_Linear
        leaveTangent=0
        leaveTangentWeight=0
        tangentMode=RCTM_Auto
        tangentWeightMode=RCTWM_WeightedNone
        time=1
        value=4
       }
      ]
     }
    }
    valueTags [1,2,3,4]
    widgetStyle {
     barThickness=4
     disabledBarImage {
      bIsDynamicallyLoaded=false
      drawAs=RoundedBox
      imageType=NoImage
      mirroring=NoMirror
      resourceName=None
      resourceObject=null
      tiling=NoTile
      imageSize=<0 0>
      margin=<ltrb 0 0 0 0>
      outlineSettings {
       bUseBrushTransparency=false
       roundingType=FixedRadius
       width=1
       color {colorUseRule=UseColor_ColorTable, specifiedColor=<rgba 0 0 0 0>}
       cornerRadii=<2 2 2 2>
      }
      tintColor {colorUseRule=UseColor_Specified, specifiedColor=<rgba 0.0048 0.0048 0.0048 1>}
      uVRegion {bIsValid=false, max=<0 0>, min=<0 0>}
     }
     disabledThumbImage {
      bIsDynamicallyLoaded=false
      drawAs=RoundedBox
      imageType=NoImage
      mirroring=NoMirror
      resourceName=None
      resourceObject=null
      tiling=NoTile
      imageSize=<12 12>
      margin=<ltrb 0 0 0 0>
      outlineSettings {
       bUseBrushTransparency=false
       roundingType=HalfHeightRadius
       width=0
       color {colorUseRule=UseColor_Specified, specifiedColor=<rgba 0 0 0 0>}
       cornerRadii=<0 0 0 0>
      }
      tintColor {colorUseRule=UseColor_Specified, specifiedColor=<rgba 0.5271 0.5271 0.5271 1>}
      uVRegion {bIsValid=false, max=<0 0>, min=<0 0>}
     }
     hoveredBarImage {
      bIsDynamicallyLoaded=false
      drawAs=RoundedBox
      imageType=NoImage
      mirroring=NoMirror
      resourceName=None
      resourceObject=null
      tiling=NoTile
      imageSize=<0 0>
      margin=<ltrb 0 0 0 0>
      outlineSettings {
       bUseBrushTransparency=false
       roundingType=FixedRadius
       width=1
       color {colorUseRule=UseColor_ColorTable, specifiedColor=<rgba 0 0 0 0>}
       cornerRadii=<2 2 2 2>
      }
      tintColor {colorUseRule=UseColor_Specified, specifiedColor=<rgba 0.0048 0.0048 0.0048 1>}
      uVRegion {bIsValid=false, max=<0 0>, min=<0 0>}
     }
     hoveredThumbImage {
      bIsDynamicallyLoaded=false
      drawAs=RoundedBox
      imageType=NoImage
      mirroring=NoMirror
      resourceName=None
      resourceObject=null
      tiling=NoTile
      imageSize=<12 12>
      margin=<ltrb 0 0 0 0>
      outlineSettings {
       bUseBrushTransparency=false
       roundingType=HalfHeightRadius
       width=0
       color {colorUseRule=UseColor_Specified, specifiedColor=<rgba 0 0 0 0>}
       cornerRadii=<0 0 0 0>
      }
      tintColor {colorUseRule=UseColor_Specified, specifiedColor=<rgba 1 1 1 1>}
      uVRegion {bIsValid=false, max=<0 0>, min=<0 0>}
     }
     normalBarImage {
      bIsDynamicallyLoaded=false
      drawAs=RoundedBox
      imageType=NoImage
      mirroring=NoMirror
      resourceName=None
      resourceObject=null
      tiling=NoTile
      imageSize=<0 0>
      margin=<ltrb 0 0 0 0>
      outlineSettings {
       bUseBrushTransparency=false
       roundingType=FixedRadius
       width=1
       color {colorUseRule=UseColor_ColorTable, specifiedColor=<rgba 0 0 0 0>}
       cornerRadii=<2 2 2 2>
      }
      tintColor {colorUseRule=UseColor_Specified, specifiedColor=<rgba 0.0048 0.0048 0.0048 1>}
      uVRegion {bIsValid=false, max=<0 0>, min=<0 0>}
     }
     normalThumbImage {
      bIsDynamicallyLoaded=false
      drawAs=RoundedBox
      imageType=NoImage
      mirroring=NoMirror
      resourceName=None
      resourceObject=null
      tiling=NoTile
      imageSize=<12 12>
      margin=<ltrb 0 0 0 0>
      outlineSettings {
       bUseBrushTransparency=false
       roundingType=HalfHeightRadius
       width=0
       color {colorUseRule=UseColor_Specified, specifiedColor=<rgba 0 0 0 0>}
       cornerRadii=<0 0 0 0>
      }
      tintColor {colorUseRule=UseColor_Specified, specifiedColor=<rgba 0.5271 0.5271 0.5271 1>}
      uVRegion {bIsValid=false, max=<0 0>, min=<0 0>}
     }
    }
   }
   slot {class=OverlaySlot, properties {horizontalAlignment=HAlign_Fill, verticalAlignment=VAlign_Fill, padding=<ltrb 0 15 0 0>}}
  }
  TextBlock_56 {
   widgetClassPath=&TextBlock
   parent=Overlay_145
   widget=TextBlock_56
   properties {
    justification=Center
    text="Duration\r\n(Beats)"
    toolTipText=Duration of phase change in beats.
    font {
     bForceMonospaced=false
     bMaterialIsStencil=false
     fontMaterial=null
     fontObject=&Roboto
     letterSpacing=0
     monospacedWidth=1
     size=9
     skewAmount=0
     typefaceFontName=Regular
     outlineSettings {
      bApplyOutlineToDropShadows=false
      bMiteredCorners=false
      bSeparateFillAlpha=false
      outlineMaterial=null
      outlineSize=0
      outlineColor=<rgba 0 0 0 1>
     }
    }
   }
   slot {class=OverlaySlot, properties {horizontalAlignment=HAlign_Center, verticalAlignment=VAlign_Center, padding=<ltrb 0 10 0 0>}}
  }
  Divider {
   widgetClassPath=&Image
   parent=VerticalBox_237
   widget=Divider
   properties {
    brush {
     bIsDynamicallyLoaded=false
     drawAs=Image
     imageType=NoImage
     mirroring=NoMirror
     resourceName=None
     resourceObject=null
     tiling=NoTile
     imageSize=<32 1>
     margin=<ltrb 0 0 0 0>
     outlineSettings {
      bUseBrushTransparency=false
      roundingType=HalfHeightRadius
      width=0
      color {colorUseRule=UseColor_Specified, specifiedColor=<rgba 0 0 0 0>}
      cornerRadii=<0 0 0 0>
     }
     tintColor {colorUseRule=UseColor_Specified, specifiedColor=<rgba 1 1 1 1>}
     uVRegion {bIsValid=false, max=<0 0>, min=<0 0>}
    }
    colorAndOpacity=<rgba 1 1 1 0.1739>
   }
   slot {class=VerticalBoxSlot, properties {verticalAlignment=VAlign_Top, padding=<ltrb 0 0 0 10>}}
  }
  VerticalBox_135 {
   widgetClassPath=&VerticalBox
   parent=VerticalBox_237
   widget=VerticalBox_135
   properties {toolTipText=Applies exponential curve to signal.}
   slot {class=VerticalBoxSlot, properties {verticalAlignment=VAlign_Top}}
  }
  HorizontalBox_155 {
   widgetClassPath=&HorizontalBox
   parent=VerticalBox_135
   widget=HorizontalBox_155
   slot {class=VerticalBoxSlot, properties {verticalAlignment=VAlign_Top, padding=<ltrb 0 0 0 5>, size {sizeRule=Fill, value=1}}}
  }
  TextBlock_237 {
   widgetClassPath=&TextBlock
   parent=HorizontalBox_155
   widget=TextBlock_237
   properties {
    justification=Right
    text=Out
    font {
     bForceMonospaced=false
     bMaterialIsStencil=false
     fontMaterial=null
     fontObject=&Roboto
     letterSpacing=0
     monospacedWidth=1
     size=7.5
     skewAmount=0
     typefaceFontName=Light
     outlineSettings {
      bApplyOutlineToDropShadows=false
      bMiteredCorners=false
      bSeparateFillAlpha=false
      outlineMaterial=null
      outlineSize=0
      outlineColor=<rgba 0 0 0 1>
     }
    }
   }
   slot {class=HorizontalBoxSlot, properties {horizontalAlignment=HAlign_Center, verticalAlignment=VAlign_Center}}
  }
  EasingSlider {
   widgetClassPath=/Script/Blutility.EditorUtilitySlider
   bIsVariable=true
   parent=HorizontalBox_155
   widget=EasingSlider
   properties {
    stepSize=0.05
    toolTipText=0
    value=0.5
    widgetStyle {
     barThickness=8
     disabledBarImage {
      bIsDynamicallyLoaded=false
      drawAs=RoundedBox
      imageType=NoImage
      mirroring=NoMirror
      resourceName=None
      resourceObject=null
      tiling=NoTile
      imageSize=<0 0>
      margin=<ltrb 0 0 0 0>
      outlineSettings {
       bUseBrushTransparency=false
       roundingType=FixedRadius
       width=1
       color {colorUseRule=UseColor_ColorTable, specifiedColor=<rgba 0 0 0 0>}
       cornerRadii=<2 2 2 2>
      }
      tintColor {colorUseRule=UseColor_Specified, specifiedColor=<rgba 0.0048 0.0048 0.0048 1>}
      uVRegion {bIsValid=false, max=<0 0>, min=<0 0>}
     }
     disabledThumbImage {
      bIsDynamicallyLoaded=false
      drawAs=RoundedBox
      imageType=NoImage
      mirroring=NoMirror
      resourceName=None
      resourceObject=null
      tiling=NoTile
      imageSize=<12 12>
      margin=<ltrb 0 0 0 0>
      outlineSettings {
       bUseBrushTransparency=false
       roundingType=HalfHeightRadius
       width=0
       color {colorUseRule=UseColor_Specified, specifiedColor=<rgba 0 0 0 0>}
       cornerRadii=<0 0 0 0>
      }
      tintColor {colorUseRule=UseColor_Specified, specifiedColor=<rgba 0.5271 0.5271 0.5271 1>}
      uVRegion {bIsValid=false, max=<0 0>, min=<0 0>}
     }
     hoveredBarImage {
      bIsDynamicallyLoaded=false
      drawAs=RoundedBox
      imageType=NoImage
      mirroring=NoMirror
      resourceName=None
      resourceObject=null
      tiling=NoTile
      imageSize=<0 0>
      margin=<ltrb 0 0 0 0>
      outlineSettings {
       bUseBrushTransparency=false
       roundingType=FixedRadius
       width=1
       color {colorUseRule=UseColor_Specified, specifiedColor=<rgba 0.0312 0.0312 0.0312 0>}
       cornerRadii=<2 2 2 2>
      }
      tintColor {colorUseRule=UseColor_Specified, specifiedColor=<rgba 0.0048 0.0048 0.0048 0.5>}
      uVRegion {bIsValid=false, max=<0 0>, min=<0 0>}
     }
     hoveredThumbImage {
      bIsDynamicallyLoaded=false
      drawAs=RoundedBox
      imageType=NoImage
      mirroring=NoMirror
      resourceName=None
      resourceObject=null
      tiling=NoTile
      imageSize=<12 12>
      margin=<ltrb 0 0 0 0>
      outlineSettings {
       bUseBrushTransparency=false
       roundingType=HalfHeightRadius
       width=0
       color {colorUseRule=UseColor_Specified, specifiedColor=<rgba 0 0 0 0>}
       cornerRadii=<0 0 0 0>
      }
      tintColor {colorUseRule=UseColor_Specified, specifiedColor=<rgba 1 1 1 1>}
      uVRegion {bIsValid=false, max=<0 0>, min=<0 0>}
     }
     normalBarImage {
      bIsDynamicallyLoaded=false
      drawAs=RoundedBox
      imageType=NoImage
      mirroring=NoMirror
      resourceName=None
      resourceObject=null
      tiling=NoTile
      imageSize=<0 0>
      margin=<ltrb 0 0 0 0>
      outlineSettings {
       bUseBrushTransparency=true
       roundingType=FixedRadius
       width=1
       color {colorUseRule=UseColor_Specified, specifiedColor=<rgba 0 0 0 0>}
       cornerRadii=<2 2 2 2>
      }
      tintColor {colorUseRule=UseColor_Specified, specifiedColor=<rgba 0 0 0 0.5>}
      uVRegion {bIsValid=false, max=<0 0>, min=<0 0>}
     }
     normalThumbImage {
      bIsDynamicallyLoaded=false
      drawAs=RoundedBox
      imageType=NoImage
      mirroring=NoMirror
      resourceName=None
      resourceObject=null
      tiling=NoTile
      imageSize=<12 12>
      margin=<ltrb 0 0 0 0>
      outlineSettings {
       bUseBrushTransparency=false
       roundingType=HalfHeightRadius
       width=0
       color {colorUseRule=UseColor_Specified, specifiedColor=<rgba 0 0 0 0>}
       cornerRadii=<0 0 0 0>
      }
      tintColor {colorUseRule=UseColor_Specified, specifiedColor=<rgba 0.5271 0.5271 0.5271 1>}
      uVRegion {bIsValid=false, max=<0 0>, min=<0 0>}
     }
    }
   }
   slot {class=HorizontalBoxSlot, properties {size {sizeRule=Fill, value=1}}}
  }
  TextBlock_1 {
   widgetClassPath=&TextBlock
   parent=HorizontalBox_155
   widget=TextBlock_1
   properties {
    text=In
    font {
     bForceMonospaced=false
     bMaterialIsStencil=false
     fontMaterial=null
     fontObject=&Roboto
     letterSpacing=0
     monospacedWidth=1
     size=7.5
     skewAmount=0
     typefaceFontName=Light
     outlineSettings {
      bApplyOutlineToDropShadows=false
      bMiteredCorners=false
      bSeparateFillAlpha=false
      outlineMaterial=null
      outlineSize=0
      outlineColor=<rgba 0 0 0 1>
     }
    }
   }
   slot {class=HorizontalBoxSlot, properties {horizontalAlignment=HAlign_Center, verticalAlignment=VAlign_Center, padding=<ltrb 2 0 0 0>}}
  }
  TextBlock {
   widgetClassPath=&TextBlock
   parent=VerticalBox_135
   widget=TextBlock
   properties {
    text=Easing
    font {
     bForceMonospaced=false
     bMaterialIsStencil=false
     fontMaterial=null
     fontObject=&Roboto
     letterSpacing=0
     monospacedWidth=1
     size=9
     skewAmount=0
     typefaceFontName=Regular
     outlineSettings {
      bApplyOutlineToDropShadows=false
      bMiteredCorners=false
      bSeparateFillAlpha=false
      outlineMaterial=null
      outlineSize=0
      outlineColor=<rgba 0 0 0 1>
     }
    }
   }
   slot {class=VerticalBoxSlot, properties {horizontalAlignment=HAlign_Center, verticalAlignment=VAlign_Top}}
  }
 ]
}
