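# Editor utility widget in the Sparks Stage Show v2 prototype event editor for shaping a beat signal with a visual preview.
# Child widgets for frequency/offset, rise, fall and scale report changes to event handlers that store values.
# UpdatePreviewImage packs these values into two vector parameters, plus an AnimationMode scalar, on the preview image's dynamic material.
# GetSignalParameters returns the shape as StageShowSignalData; SetAnimationMode also sets the Fall controls' enabled state.
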EditorUtilityWidgetBlueprint /SparksStageShow/Systems/StageManager_v2_Prototype/Tools/EventEditor/Sparks_StageShowUtility_BeatSignalEditor
graphs [
 EventGraph {nodes [
   n1 = OnFallParamsChange!() -> (OutputDelegate, then, Duration:double=~, Easing:double=~) {is=title K2Node_CustomEvent}
   > Set FallDuration(FallDuration:double=@n1.Duration) -> (Output_Get:double)
   > n2 = Set FallEasing(FallEasing:double=@n1.Easing) -> (Output_Get:double)
   UpdatePreviewImage(execute=[@n2.then, @n4.then, @n10.then, @n12.then])
   n3 = OnRiseParamsChange!() -> (OutputDelegate, then, Duration:double=~, Easing:double=~) {is=title K2Node_CustomEvent}
   > Set RiseDuration(RiseDuration:double=@n3.Duration) -> (Output_Get:double)
   > n4 = Set RiseEasing(RiseEasing:double=@n3.Easing) -> (Output_Get:double)
   UpdatePreviewImage() {is=title K2Node_CustomEvent}
   n5 = "Set Vector Parameter MID (Brush)"!(
    Brush:Slate Brush Structure=Get Brush!(self:Image Object Reference=Get SignalPreviewImage() -> (SignalPreviewImage:Image Object Reference)) -> (Brush:Slate Brush Structure)
    Material:Material Interface Object Reference=Get PreviewImageMID() -> (PreviewImageMID:Material Instance Dynamic Object Reference)
    ParameterName:Name=SignalProfile
    Value:Linear Color Structure=Make Linear Color!(
     R:float=Get RiseDuration() -> (RiseDuration:double)
     G:float=Get RiseEasing() -> (RiseEasing:double)
     B:float=Get FallDuration() -> (FallDuration:double)
     A:float=Get FallEasing() -> (FallEasing:double)
    ) -> (LinearColor:Linear Color Structure) {is=title K2Node_MakeStruct}
   ) -> (ReturnValue:Slate Brush Structure) {is=title K2Node_CallFunction}
   n6 = Make Linear Color(
    R=Get BeatMultiplier() -> (BeatMultiplier:double)
    G=Get Offset() -> (Offset:double)
    B=Get MaxValue() -> (MaxValue:double)
    A=Get MinValue() -> (MinValue:double)
   )
   n7 = Get PreviewImageMID()
   > SetBrush(
    self=Get SignalPreviewImage()
    InBrush="Set Scalar Parameter MID (Brush)"!(
     Brush:Slate Brush Structure="Set Vector Parameter MID (Brush)"(Brush=@n5, Material=@n7, ParameterName=SignalScaleOffset, Value=@n6)
     Material:Material Interface Object Reference=Get PreviewImageMID()
     ParameterName:Name=AnimationMode
     Value:float=Get AnimationMode() -> (AnimationMode:double)
    ) -> (ReturnValue:Slate Brush Structure) {is=title K2Node_CallFunction}
   )
   n8 = SetAnimationMode!() -> (OutputDelegate, then, AnimationMode:Byte=NewEnumerator0) {is=title K2Node_CustomEvent}
   > Set AnimationMode(AnimationMode:double=Select(Option 0:double=0, Option 1:double=0, Index:Byte=@n8.AnimationMode) -> (ReturnValue:double)) -> (Output_Get:double)
   > SetIsEnabled(
    self=Get SignalPhaseParams_Fall() -> (SignalPhaseParams_Fall:Sparks Stage Show Utility Signal Phase Params Object Reference)
    bInIsEnabled=Select(Option 0:bool=false, Option 1:bool=false, Index=@n8.AnimationMode) -> (ReturnValue:bool)
   )
   > UpdatePreviewImage() {is=title K2Node_CallFunction}
   n9 = OnSignalScaleChange!() -> (OutputDelegate, then, Min:double=~, Max:double=~) {is=title K2Node_CustomEvent}
   > Set MinValue(MinValue:double=@n9.Min) -> (Output_Get:double)
   > n10 = Set MaxValue(MaxValue:double=@n9.Max) -> (Output_Get:double)
   n11 = OnOffsetValueChanged!() -> (OutputDelegate, then, BeatMultiplier:double=~, Offset:double=~) {is=title K2Node_CustomEvent}
   > Set BeatMultiplier(BeatMultiplier:double=@n11.BeatMultiplier) -> (Output_Get:double)
   > n12 = Set Offset(Offset:double=@n11.Offset) -> (Output_Get:double)
   Event PreConstruct()
   n13 = Create Event()
   n14 = Get SignalPhaseParams_Fall()
   > Bind Event to On Value Changed(self:Sparks Stage Show Utility Signal Phase Params Object Reference=@n14, Delegate=@n13) {is=title K2Node_AddDelegate}
   n15 = Create Event()
   n16 = Get SignalPhaseParams_Rise() -> (SignalPhaseParams_Rise:Sparks Stage Show Utility Signal Phase Params Object Reference)
   > Bind Event to On Value Changed(self=@n16, Delegate=@n15)
   n17 = Create Event()
   n18 = Get SignalScaleWidget() -> (SignalScaleWidget:Sparks Stage Show Utility Signal Scale Object Reference)
   > Bind Event to On Value Changed(self:Sparks Stage Show Utility Signal Scale Object Reference=@n18, Delegate=@n17)
   n19 = Create Event()
   n20 = Get SignalOffsetRadialWidget() -> (SignalOffsetRadialWidget:Sparks Stage Show Utility Signal Offset Radial Object Reference)
   > Bind Event to On Value Changed(self:Sparks Stage Show Utility Signal Offset Radial Object Reference=@n20, Delegate=@n19)
   > n21 = Is Valid(InputObject=Get PreviewImageMID())
   n22 = CreateDynamicMaterialInstance(
    execute=@n21.Is Not Valid
    WorldContextObject=Self-Reference()
    Parent=/SparksStageShow/Systems/StageManager_v2_Prototype/Tools/EventEditor/Materials/M_StageShowUtility_SignalPreview
    OptionalName=None
    CreationFlags=None
   )
   > Set PreviewImageMID(PreviewImageMID:Material Instance Dynamic Object Reference=@n22) -> (Output_Get:Material Instance Dynamic Object Reference)
   Event Destruct()
   > Unbind Event from On Value Changed(self:Sparks Stage Show Utility Signal Phase Params Object Reference=@n14, Delegate=@n13) {is=title K2Node_RemoveDelegate}
   > Unbind Event from On Value Changed(self=@n16, Delegate=@n15)
   > Unbind Event from On Value Changed(self:Sparks Stage Show Utility Signal Scale Object Reference=@n18, Delegate=@n17)
   > Unbind Event from On Value Changed(self:Sparks Stage Show Utility Signal Offset Radial Object Reference=@n20, Delegate=@n19)
   Event Construct()
   Event Tick()
  ]}
 GetSignalParameters {nodes [
   GetSignalParameters() {is=title K2Node_FunctionEntry}
   > Return Node!(
    execute
    Signal Data:Stage Show Signal Data Structure=Make Stage Show Signal Data!(
     RiseDuration:double=Get RiseDuration()
     RiseEasing:double=Get RiseEasing()
     FallDuration:double=Get FallDuration()
     FallEasing:double=Get FallEasing()
     BeatMultiplier:double=Get BeatMultiplier()
     BeatOffset:double=Get Offset()
     PeakValue:double=Get MaxValue()
     MinimumValue:double=Get MinValue()
    ) -> (StageShowSignalData:Stage Show Signal Data Structure) {is=title K2Node_MakeStruct}
   )
  ]}
]
variables [
 AnimationMode {type=number, default=1}
 BeatMultiplier {type=number, default=1}
 FallDuration {type=number, default=2}
 FallEasing {type=number, default=1}
 MaxValue {type=number, default=1}
 MinValue {type=number}
 Offset {type=number}
 PreviewImageMID {type=MaterialInstanceDynamic}
 RiseDuration {type=number, default=2}
 RiseEasing {type=number, default=1}
]
widgets {
 info {rootWidgetClass=&SizeBox, widgetCount=20}
 widgets [
  SizeBox_2 {
   widgetClassPath=&SizeBox
   widget=SizeBox_2
   properties {bOverride_MaxDesiredHeight=true, heightOverride=150, maxDesiredHeight=240, minDesiredHeight=60}
  }
  HorizontalBox_39 {
   widgetClassPath=&HorizontalBox
   parent=SizeBox_2
   widget=HorizontalBox_39
   slot {class=SizeBoxSlot, properties {horizontalAlignment=HAlign_Left, verticalAlignment=VAlign_Top}}
  }
  HorizontalBox_683 {
   widgetClassPath=&HorizontalBox
   parent=HorizontalBox_39
   widget=HorizontalBox_683
   slot {class=HorizontalBoxSlot, properties {horizontalAlignment=HAlign_Left, verticalAlignment=VAlign_Top}}
  }
  VerticalBox_66 {
   widgetClassPath=&VerticalBox
   parent=HorizontalBox_683
   widget=VerticalBox_66
   slot {class=HorizontalBoxSlot, properties {size {sizeRule=Fill, value=1}}}
  }
  TextBlock_3 {
   widgetClassPath=&TextBlock
   parent=VerticalBox_66
   widget=TextBlock_3
   properties {
    justification=Center
    text="Frequency && Offset"
    font {
     bForceMonospaced=false
     bMaterialIsStencil=false
     fontMaterial=null
     fontObject=&Roboto
     letterSpacing=0
     monospacedWidth=1
     size=9
     skewAmount=0
     typefaceFontName=Bold
     outlineSettings {
      bApplyOutlineToDropShadows=false
      bMiteredCorners=false
      bSeparateFillAlpha=false
      outlineMaterial=null
      outlineSize=0
      outlineColor=<rgba 0 0 0 1>
     }
    }
   }
   slot {class=VerticalBoxSlot, properties {padding=<ltrb 0 0 0 10>}}
  }
  SignalOffsetRadialWidget {
   widgetClassPath=/SparksStageShow/Systems/StageManager_v2_Prototype/Tools/EventEditor/Widgets/Sparks_StageShowUtility_SignalOffsetRadial
   bIsVariable=true
   parent=VerticalBox_66
   widget=SignalOffsetRadialWidget
   slot {class=VerticalBoxSlot, properties {size {sizeRule=Fill, value=1}}}
  }
  Divider_2 {
   widgetClassPath=&Image
   parent=HorizontalBox_39
   widget=Divider_2
   properties {brush {
     bIsDynamicallyLoaded=false
     drawAs=Image
     imageType=NoImage
     mirroring=NoMirror
     resourceName=None
     resourceObject=null
     tiling=NoTile
     imageSize=<1 32>
     margin=<ltrb 0 0 0 0>
     outlineSettings {
      bUseBrushTransparency=false
      roundingType=HalfHeightRadius
      width=0
      color {colorUseRule=UseColor_Specified, specifiedColor=<rgba 0 0 0 0>}
      cornerRadii=<0 0 0 0>
     }
     tintColor {colorUseRule=UseColor_Specified, specifiedColor=<rgba 1 1 1 0.1>}
     uVRegion {bIsValid=false, max=<0 0>, min=<0 0>}
    }}
   slot {class=HorizontalBoxSlot, properties {horizontalAlignment=HAlign_Left, padding=<ltrb 5 0 5 0>}}
  }
  VerticalBox_1 {
   widgetClassPath=&VerticalBox
   parent=HorizontalBox_39
   widget=VerticalBox_1
   slot {class=HorizontalBoxSlot, properties {horizontalAlignment=HAlign_Left, verticalAlignment=VAlign_Top}}
  }
  TextBlock_192 {
   widgetClassPath=&TextBlock
   parent=VerticalBox_1
   widget=TextBlock_192
   properties {
    justification=Center
    text=Signal Rise
    font {
     bForceMonospaced=false
     bMaterialIsStencil=false
     fontMaterial=null
     fontObject=&Roboto
     letterSpacing=0
     monospacedWidth=1
     size=9
     skewAmount=0
     typefaceFontName=Bold
     outlineSettings {
      bApplyOutlineToDropShadows=false
      bMiteredCorners=false
      bSeparateFillAlpha=false
      outlineMaterial=null
      outlineSize=0
      outlineColor=<rgba 0 0 0 1>
     }
    }
   }
   slot {class=VerticalBoxSlot, properties {padding=<ltrb 0 0 0 10>}}
  }
  SignalPhaseParams_Rise {
   widgetClassPath=/SparksStageShow/Systems/StageManager_v2_Prototype/Tools/EventEditor/Widgets/Sparks_StageShowUtility_SignalPhaseParams
   bIsVariable=true
   parent=VerticalBox_1
   widget=SignalPhaseParams_Rise
   slot {class=VerticalBoxSlot, properties {size {sizeRule=Fill, value=1}}}
  }
  Divider {
   widgetClassPath=&Image
   parent=HorizontalBox_39
   widget=Divider
   properties {brush {
     bIsDynamicallyLoaded=false
     drawAs=Image
     imageType=NoImage
     mirroring=NoMirror
     resourceName=None
     resourceObject=null
     tiling=NoTile
     imageSize=<1 32>
     margin=<ltrb 0 0 0 0>
     outlineSettings {
      bUseBrushTransparency=false
      roundingType=HalfHeightRadius
      width=0
      color {colorUseRule=UseColor_Specified, specifiedColor=<rgba 0 0 0 0>}
      cornerRadii=<0 0 0 0>
     }
     tintColor {colorUseRule=UseColor_Specified, specifiedColor=<rgba 1 1 1 0.1>}
     uVRegion {bIsValid=false, max=<0 0>, min=<0 0>}
    }}
   slot {class=HorizontalBoxSlot, properties {horizontalAlignment=HAlign_Left, padding=<ltrb 5 0 5 0>}}
  }
  VerticalBox_2 {
   widgetClassPath=&VerticalBox
   parent=HorizontalBox_39
   widget=VerticalBox_2
   slot {class=HorizontalBoxSlot, properties {horizontalAlignment=HAlign_Left, verticalAlignment=VAlign_Top}}
  }
  TextBlock_1 {
   widgetClassPath=&TextBlock
   parent=VerticalBox_2
   widget=TextBlock_1
   properties {
    justification=Center
    text=Signal Fall
    font {
     bForceMonospaced=false
     bMaterialIsStencil=false
     fontMaterial=null
     fontObject=&Roboto
     letterSpacing=0
     monospacedWidth=1
     size=9
     skewAmount=0
     typefaceFontName=Bold
     outlineSettings {
      bApplyOutlineToDropShadows=false
      bMiteredCorners=false
      bSeparateFillAlpha=false
      outlineMaterial=null
      outlineSize=0
      outlineColor=<rgba 0 0 0 1>
     }
    }
   }
   slot {class=VerticalBoxSlot, properties {padding=<ltrb 0 0 0 10>}}
  }
  SignalPhaseParams_Fall {
   widgetClassPath=&Sparks_StageShowUtility_SignalPhaseParams
   bIsVariable=true
   parent=VerticalBox_2
   widget=SignalPhaseParams_Fall
   properties {invertEasing=true}
   slot {class=VerticalBoxSlot, properties {size {sizeRule=Fill, value=1}}}
  }
  Divider_3 {
   widgetClassPath=&Image
   parent=HorizontalBox_39
   widget=Divider_3
   properties {brush {
     bIsDynamicallyLoaded=false
     drawAs=Image
     imageType=NoImage
     mirroring=NoMirror
     resourceName=None
     resourceObject=null
     tiling=NoTile
     imageSize=<1 32>
     margin=<ltrb 0 0 0 0>
     outlineSettings {
      bUseBrushTransparency=false
      roundingType=HalfHeightRadius
      width=0
      color {colorUseRule=UseColor_Specified, specifiedColor=<rgba 0 0 0 0>}
      cornerRadii=<0 0 0 0>
     }
     tintColor {colorUseRule=UseColor_Specified, specifiedColor=<rgba 1 1 1 0.1>}
     uVRegion {bIsValid=false, max=<0 0>, min=<0 0>}
    }}
   slot {class=HorizontalBoxSlot, properties {horizontalAlignment=HAlign_Left, padding=<ltrb 5 0 5 0>}}
  }
  VerticalBox_4 {
   widgetClassPath=&VerticalBox
   parent=HorizontalBox_39
   widget=VerticalBox_4
   properties {bOverride_Cursor=true}
   slot {class=HorizontalBoxSlot, properties {horizontalAlignment=HAlign_Left}}
  }
  TextBlock {
   widgetClassPath=&TextBlock
   parent=VerticalBox_4
   widget=TextBlock
   properties {
    justification=Center
    text=Signal Scale
    font {
     bForceMonospaced=false
     bMaterialIsStencil=false
     fontMaterial=null
     fontObject=&Roboto
     letterSpacing=0
     monospacedWidth=1
     size=9
     skewAmount=0
     typefaceFontName=Bold
     outlineSettings {
      bApplyOutlineToDropShadows=false
      bMiteredCorners=false
      bSeparateFillAlpha=false
      outlineMaterial=null
      outlineSize=0
      outlineColor=<rgba 0 0 0 1>
     }
    }
   }
   slot {class=VerticalBoxSlot}
  }
  SignalScaleWidget {
   widgetClassPath=/SparksStageShow/Systems/StageManager_v2_Prototype/Tools/EventEditor/Widgets/Sparks_StageShowUtility_SignalScale
   bIsVariable=true
   parent=VerticalBox_4
   widget=SignalScaleWidget
   slot {class=VerticalBoxSlot, properties {padding=<ltrb 0 10 0 0>, size {sizeRule=Fill, value=1}}}
  }
  SizeBox_39 {
   widgetClassPath=&SizeBox
   parent=HorizontalBox_39
   widget=SizeBox_39
   properties {bOverride_WidthOverride=true, heightOverride=150, widthOverride=400}
   slot {class=HorizontalBoxSlot, properties {horizontalAlignment=HAlign_Left, padding=<ltrb 20 0 0 0>}}
  }
  SignalPreviewImage {
   widgetClassPath=&Image
   bIsVariable=true
   parent=SizeBox_39
   widget=SignalPreviewImage
   properties {brush {
     bIsDynamicallyLoaded=false
     drawAs=Image
     imageType=FullColor
     mirroring=NoMirror
     resourceName=None
     resourceObject=/SparksStageShow/Systems/StageManager_v2_Prototype/Tools/EventEditor/Materials/MI_StageShowUtility_SignalPreview
     tiling=NoTile
     imageSize=<32 32>
     margin=<ltrb 0 0 0 0>
     outlineSettings {
      bUseBrushTransparency=false
      roundingType=HalfHeightRadius
      width=0
      color {colorUseRule=UseColor_Specified, specifiedColor=<rgba 0 0 0 0>}
      cornerRadii=<0 0 0 0>
     }
     tintColor {colorUseRule=UseColor_Specified, specifiedColor=<rgba 1 1 1 1>}
     uVRegion {bIsValid=false, max=<0 0>, min=<0 0>}
    }}
   slot {class=SizeBoxSlot}
  }
 ]
}
