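# Editor utility widget providing a metronome for testing stage show actors, with a BPM slider and a Play/Stop button that broadcasts OnStateChange.
# While enabled, Tick derives four-beat measure timing from BPM, playing pings on beats and measures, with lower pitch for beats.
# It writes fractional progress into MPC_SparksStageShow's BeatLifetime and MeasureLifetime parameters.
# BPM changes and toggling reset timers, but zero those parameters in a different collection, MPC_Sparks_Audio_Parameters.
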EditorUtilityWidgetBlueprint /SparksStageShow/Systems/StageManager_v2_Prototype/Tools/EventEditor/Sparks_StageShowUtility_Metronome
graphs [
 EventGraph {nodes [
   n1 = "On Value Changed (BPMSlider)"!() -> (OutputDelegate, then, Value:float) {is=title K2Node_ComponentBoundEvent}
   > n2 = Set BPM(BPM:double=@n1.Value) -> (Output_Get:double)
   > "SetText (Text)"(
    self=Get BPMValueText() -> (BPMValueText:Text Object Reference)
    InText="To Text (Float)"(Value=@n2, MinimumIntegralDigits=1, MaximumIntegralDigits=324, MinimumFractionalDigits=0, MaximumFractionalDigits=0)
   )
   > ResetMetronome()
   "On Released (PlayButton)"() {is=title K2Node_ComponentBoundEvent}
   > Set MetronomeEnabled(MetronomeEnabled:bool=NOT Boolean(A=Get MetronomeEnabled() -> (MetronomeEnabled:bool))) -> (Output_Get:bool)
   > ResetMetronome()
   n3 = Get MetronomeEnabled()
   > "SetText (Text)"(
    self=Get PlayButtonLabel() -> (PlayButtonLabel:Text Object Reference)
    InText=Select(Option 0:Text="LOCTEXT(\"Play\")", Option 1:Text="LOCTEXT(\"Stop\")", Index=@n3) -> (ReturnValue:Text)
   )
   > SetBrushTintColor!(
    execute
    self:Image Object Reference=Get PlayButtonIcon() -> (PlayButtonIcon:Image Object Reference)
    TintColor:Slate Color Structure=Select(
     Option 0:Slate Color Structure=Make Slate Color(SpecifiedColor=(R=1,G=1,B=1,A=1))
     Option 1:Slate Color Structure=Make Slate Color(SpecifiedColor=(R=0.1837,G=1,B=0.1968,A=1))
     Index=@n3
    ) -> (ReturnValue:Slate Color Structure)
   ) -> (then)
   > Call On State Change!(execute:Exec, self:Sparks Stage Show Utility Metronome Object Reference, IsPlaying:bool=Get MetronomeEnabled()) -> (then:Exec) {is=title K2Node_CallDelegate}
   Event PreConstruct()
   > "SetText (Text)"(
    self=Get BPMValueText()
    InText="To Text (Float)"(Value=Get BPM() -> (BPM:double), MinimumIntegralDigits=1, MaximumIntegralDigits=324, MinimumFractionalDigits=0, MaximumFractionalDigits=0)
   )
   > SetValue!(execute, self:Slider Object Reference=Get BPMSlider() -> (BPMSlider:Editor Utility Slider Object Reference), InValue:float=Get BPM()) -> (then)
   n4 = Event Tick()
   > Branch(Condition=Get MetronomeEnabled())
   > n5 = Sequence!(execute) -> (then_0, then_1, then_2:Exec)
   n6 = Get BPM()
   n7 = float / float(A=1, B=float / float(A=float / float(A=@n6, B=4), B=60))
   > Set MeasureLifetime(MeasureLifetime:double=@n7) -> (Output_Get:double)
   > Set BeatLifetime(BeatLifetime:double=float / float(A=@n7, B=4)) -> (Output_Get:double)
   n8 = Set BeatTimer(execute=@n5.then_1, BeatTimer:double=float + float(A=@n4.InDeltaTime, B=Get BeatTimer() -> (BeatTimer:double))) -> (Output_Get:double)
   > n9 = Set BeatProgression(BeatProgression:double=float / float(A=@n8, B=Get BeatLifetime() -> (BeatLifetime:double))) -> (Output_Get:double)
   > SetScalarParameterValue!(
    execute
    WorldContextObject
    Collection:Material Parameter Collection Object Reference=/SparksStageShow/Systems/StageManager_v2_Prototype/MaterialFunctions/MPC_SparksStageShow
    ParameterName:Name=BeatLifetime
    ParameterValue:float=Fraction!(A:double=@n9) -> (ReturnValue:double)
   ) -> (then) {is=title K2Node_CallMaterialParameterCollectionFunction}
   n10 = Get MeasureProgression() -> (MeasureProgression:double)
   n11 = Get BeatProgression() -> (BeatProgression:double)
   > Branch(Condition=AND Boolean(A="float >= float"(A=@n11, B=1), B="float < float"(A=@n10, B=1)))
   > Set BeatTimer(BeatTimer=float - float(A=Get BeatTimer(), B=Get BeatLifetime()))
   > PlaySound2D!(
    execute
    WorldContextObject:Object Reference=Self-Reference()
    Sound=/Engine/EngineSounds/1kSineTonePing
    VolumeMultiplier
    PitchMultiplier=0.5
    StartTime
    ConcurrencySettings
    OwningActor
    bIsUISound
   ) -> (then)
   n12 = Set MeasureTimer(execute=@n5.then_2, MeasureTimer:double=float + float(A=@n4.InDeltaTime, B=Get MeasureTimer() -> (MeasureTimer:double))) -> (Output_Get:double)
   > n13 = Set MeasureProgression(MeasureProgression:double=float / float(A=@n12, B=Get MeasureLifetime() -> (MeasureLifetime:double))) -> (Output_Get:double)
   > SetScalarParameterValue(
    Collection=&MPC_SparksStageShow
    ParameterName=MeasureLifetime
    ParameterValue=Fraction(A=@n13)
   )
   > Branch(Condition="float >= float"(A=Get MeasureProgression(), B=1))
   > Set MeasureTimer(MeasureTimer=float - float(A=Get MeasureTimer(), B=Get MeasureLifetime()))
   > PlaySound2D(WorldContextObject=Self-Reference(), Sound=&1kSineTonePing)
   Event Construct()
  ]}
 OnStateChange {nodes [OnStateChange!() -> (then, IsPlaying:bool=~) {is=title K2Node_FunctionEntry}]}
 ResetMetronome {nodes [
   ResetMetronome() {is=title K2Node_FunctionEntry}
   > Set BeatTimer(BeatTimer=0)
   > Set MeasureTimer(MeasureTimer=0)
   > Set BeatProgression(BeatProgression=0)
   > Set MeasureProgression(MeasureProgression=0)
   > SetScalarParameterValue(Collection=/FMAudioReactiveData/AudioAnalysis/MPC_Sparks_Audio_Parameters, ParameterName=BeatLifetime, ParameterValue=0)
   > SetScalarParameterValue(Collection=&MPC_Sparks_Audio_Parameters, ParameterName=MeasureLifetime, ParameterValue=0)
  ]}
]
variables [
 BeatLifetime {type=number}
 BeatProgression {type=number}
 BeatTimer {type=number}
 BPM {type=number, default=120}
 MeasureLifetime {type=number}
 MeasureProgression {type=number}
 MeasureTimer {type=number}
 MetronomeEnabled {type=boolean}
 OnStateChange {default=()}
]
widgets {
 info {rootWidgetClass=&SizeBox, widgetCount=13}
 widgets [
  SizeBox_0 {
   widgetClassPath=&SizeBox
   widget=SizeBox_0
   properties {bOverride_MinDesiredHeight=true, bOverride_MinDesiredWidth=true, maxDesiredHeight=300, minDesiredHeight=40, minDesiredWidth=100}
  }
  HorizontalBox_0 {
   widgetClassPath=&HorizontalBox
   parent=SizeBox_0
   widget=HorizontalBox_0
   slot {class=SizeBoxSlot, properties {padding=<ltrb 0 5 0 5>}}
  }
  HorizontalBox_25 {
   widgetClassPath=&HorizontalBox
   parent=HorizontalBox_0
   widget=HorizontalBox_25
   slot {class=HorizontalBoxSlot, properties {padding=<ltrb 0 0 10 0>, size {sizeRule=Fill, value=1}}}
  }
  TextBlock_30 {
   widgetClassPath=&TextBlock
   parent=HorizontalBox_25
   widget=TextBlock_30
   properties {
    text=BPM
    font {
     bForceMonospaced=false
     bMaterialIsStencil=false
     fontMaterial=null
     fontObject=&Roboto
     letterSpacing=0
     monospacedWidth=1
     size=9
     skewAmount=0
     typefaceFontName=Regular
     outlineSettings {
      bApplyOutlineToDropShadows=false
      bMiteredCorners=false
      bSeparateFillAlpha=false
      outlineMaterial=null
      outlineSize=0
      outlineColor=<rgba 0 0 0 1>
     }
    }
   }
   slot {class=HorizontalBoxSlot, properties {verticalAlignment=VAlign_Center, padding=<ltrb 0 0 5 0>}}
  }
  BPMSlider {
   widgetClassPath=/Script/Blutility.EditorUtilitySlider
   bIsVariable=true
   parent=HorizontalBox_25
   widget=BPMSlider
   properties {maxValue=200, minValue=1, mouseUsesStep=true, stepSize=1, value=90}
   slot {class=HorizontalBoxSlot, properties {size {sizeRule=Fill, value=1}}}
  }
  BPMValueText {
   widgetClassPath=&TextBlock
   bIsVariable=true
   parent=HorizontalBox_25
   widget=BPMValueText
   properties {
    text=120
    font {
     bForceMonospaced=false
     bMaterialIsStencil=false
     fontMaterial=null
     fontObject=&Roboto
     letterSpacing=0
     monospacedWidth=1
     size=9
     skewAmount=0
     typefaceFontName=Regular
     outlineSettings {
      bApplyOutlineToDropShadows=false
      bMiteredCorners=false
      bSeparateFillAlpha=false
      outlineMaterial=null
      outlineSize=0
      outlineColor=<rgba 0 0 0 1>
     }
    }
   }
   slot {class=HorizontalBoxSlot, properties {verticalAlignment=VAlign_Center, padding=<ltrb 5 0 0 0>}}
  }
  SizeBox_1 {
   widgetClassPath=&SizeBox
   parent=HorizontalBox_0
   widget=SizeBox_1
   properties {bOverride_MinDesiredWidth=true, maxDesiredWidth=20, minDesiredWidth=100, widthOverride=200}
   slot {class=HorizontalBoxSlot}
  }
  Overlay_0 {widgetClassPath=&Overlay, parent=SizeBox_1, widget=Overlay_0, slot {class=SizeBoxSlot}}
  PlayButton {
   widgetClassPath=&EditorUtilityButton
   parent=Overlay_0
   widget=PlayButton
   properties {toolTipText=Activates a metronome to test the actors behavior.}
   slot {class=OverlaySlot, properties {horizontalAlignment=HAlign_Fill, verticalAlignment=VAlign_Fill}}
  }
  HorizontalBox_1 {
   widgetClassPath=&HorizontalBox
   parent=Overlay_0
   widget=HorizontalBox_1
   properties {visibility=HitTestInvisible}
   slot {class=OverlaySlot, properties {horizontalAlignment=HAlign_Fill, verticalAlignment=VAlign_Center, padding=<ltrb 10 5 10 5>}}
  }
  ScaleBox_0 {
   widgetClassPath=&ScaleBox
   parent=HorizontalBox_1
   widget=ScaleBox_0
   slot {class=HorizontalBoxSlot, properties {horizontalAlignment=HAlign_Left, verticalAlignment=VAlign_Center}}
  }
  PlayButtonIcon {
   widgetClassPath=&Image
   bIsVariable=true
   parent=ScaleBox_0
   widget=PlayButtonIcon
   properties {
    brush {
     bIsDynamicallyLoaded=false
     drawAs=Image
     imageType=FullColor
     mirroring=NoMirror
     resourceName=None
     resourceObject=null
     tiling=NoTile
     imageSize=<14 14>
     margin=<ltrb 0 0 0 0>
     outlineSettings {
      bUseBrushTransparency=false
      roundingType=HalfHeightRadius
      width=0
      color {colorUseRule=UseColor_Specified, specifiedColor=<rgba 0 0 0 0>}
      cornerRadii=<0 0 0 0>
     }
     tintColor {colorUseRule=UseColor_Specified, specifiedColor=<rgba 1 1 1 1>}
     uVRegion {bIsValid=false, max=<0 0>, min=<0 0>}
    }
    renderTransform {angle=90, scale=<1 1>, shear=<0 0>, translation=<0 0>}
   }
   slot {class=ScaleBoxSlot, properties {horizontalAlignment=HAlign_Left}}
  }
  PlayButtonLabel {
   widgetClassPath=&TextBlock
   bIsVariable=true
   parent=HorizontalBox_1
   widget=PlayButtonLabel
   properties {
    text=Play
    font {
     bForceMonospaced=false
     bMaterialIsStencil=false
     fontMaterial=null
     fontObject=&Roboto
     letterSpacing=0
     monospacedWidth=1
     size=9
     skewAmount=0
     typefaceFontName=Bold
     outlineSettings {
      bApplyOutlineToDropShadows=false
      bMiteredCorners=false
      bSeparateFillAlpha=false
      outlineMaterial=null
      outlineSize=0
      outlineColor=<rgba 0 0 0 1>
     }
    }
   }
   slot {class=HorizontalBoxSlot, properties {horizontalAlignment=HAlign_Left, verticalAlignment=VAlign_Center, padding=<ltrb 5 0 0 0>}}
  }
 ]
}
