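# Editor utility widget prototype for authoring stage show events and playing them on a selected actor group in the level.
# On first PreConstruct it fills enum-driven combo boxes, finds and initializes actor groups, and binds a metronome handler that is only a stub.
# Apply packs intensity and rotation beat signals, rotation settings and group offset into event settings for PlayEvent.
# Palette color edits write directly to a Material Parameter Collection.
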EditorUtilityWidgetBlueprint /SparksStageShow/Systems/StageManager_v2_Prototype/Tools/EventEditor/Sparks_StageShowUtility_EventEditor
graphs [
 EventGraph {nodes [
   n1 = "On Selection Changed (RotationPlaybackCBox)"!() -> (OutputDelegate, then, SelectedItem:String, SelectionType:ESelectInfo Enum) {is=title K2Node_ComponentBoundEvent}
   > n2 = Switch on ESelectInfo!(execute, Selection:ESelectInfo Enum=@n1.SelectionType) -> (OnKeyPress:Exec, OnNavigation:Exec, OnMouseClick:Exec, Direct:Exec) {is=title K2Node_SwitchEnum}
   n3 = ForEach EStageShowRotationMode!(execute:Exec=[@n2.Direct, @n2.OnKeyPress, @n2.OnMouseClick], SkipHidden:bool) -> (LoopBody:Exec, EnumValue:EStageShowRotationMode Enum, then:Exec) {is=title K2Node_ForEachElementInEnum}
   > Branch(Condition="Equal Exactly (String)"(A=Enum to String!(Enumerator:EStageShowRotationMode Enum=@n3) -> (ReturnValue:String), B=@n1.SelectedItem))
   > SetSelectedOption!(
    execute
    self:"Combo Box (String) Object Reference"=Get RotationPlaybackCBox() -> (RotationPlaybackCBox:Editor Utility Combo Box String Object Reference)
    Option:String=@n1.SelectedItem
   ) -> (then)
   > n4 = Set RotationPlaybackType(RotationPlaybackType:EStageShowRotationMode Enum=@n3) -> (Output_Get:EStageShowRotationMode Enum=PingPong)
   > SetAnimationMode!(
    execute
    self:Sparks Stage Show Utility Beat Signal Editor Object Reference=Get BeatSignalEditor_Rotation() -> (BeatSignalEditor_Rotation:Sparks Stage Show Utility Beat Signal Editor Object Reference)
    AnimationMode:Byte=@n4
   ) -> (then)
   n5 = "On Selection Changed (RotationPatternCBox)"!() -> (OutputDelegate, then, SelectedItem:String, SelectionType:ESelectInfo Enum) {is=title K2Node_ComponentBoundEvent}
   > n6 = Switch on ESelectInfo(Selection=@n5.SelectionType)
   n7 = ForEach EStageShowRotationPattern!(execute:Exec=[@n6.Direct, @n6.OnKeyPress, @n6.OnMouseClick], SkipHidden:bool) -> (LoopBody:Exec, EnumValue:EStageShowRotationPattern Enum, then:Exec) {is=title K2Node_ForEachElementInEnum}
   SetIsEnabled(
    execute=@n7.then
    self=[Get BeatSignalEditor_Rotation(), Get RotationPlaybackCBox()]
    bInIsEnabled="Not Equal (Enum)"!(
     A:EStageShowRotationPattern Enum=Get RotationPattern() -> (RotationPattern:EStageShowRotationPattern Enum=None)
     B:EStageShowRotationPattern Enum=None
    ) -> (ReturnValue:bool)
   )
   Branch(
    execute=@n7.LoopBody
    Condition="Equal Exactly (String)"(A=Enum to String(Enumerator:EStageShowRotationPattern Enum=@n7), B=@n5.SelectedItem)
   )
   > SetSelectedOption(self=Get RotationPatternCBox() -> (RotationPatternCBox:Editor Utility Combo Box String Object Reference), Option=@n5.SelectedItem)
   > Set RotationPattern(RotationPattern:EStageShowRotationPattern Enum=@n7) -> (Output_Get:EStageShowRotationPattern Enum=None)
   n8 = "On Value Changed (GroupPatternDurationSlider)"!() -> (OutputDelegate, then, Value:float) {is=title K2Node_ComponentBoundEvent}
   > n9 = Set GroupOffsetDuration(GroupOffsetDuration:double=@n8.Value) -> (Output_Get:double)
   > SetToolTipText!(
    execute
    self:Widget Object Reference=Get GroupPatternDurationSlider() -> (GroupPatternDurationSlider:Editor Utility Slider Object Reference)
    InToolTipText:"Text (by ref)"="To Text (Float)"(Value=@n9, MinimumIntegralDigits=1, MaximumIntegralDigits=324, MinimumFractionalDigits=0, MaximumFractionalDigits=2)
   ) -> (then)
   n10 = "On Selection Changed (GroupPatternCBox)"!() -> (OutputDelegate, then, SelectedItem:String, SelectionType:ESelectInfo Enum) {is=title K2Node_ComponentBoundEvent}
   > n11 = Switch on ESelectInfo(Selection=@n10.SelectionType)
   n12 = ForEach EStageGroupOffset!(execute:Exec=[@n11.Direct, @n11.OnKeyPress, @n11.OnMouseClick], SkipHidden:bool) -> (LoopBody:Exec, EnumValue:EStageGroupOffset Enum, then:Exec) {is=title K2Node_ForEachElementInEnum}
   SetIsEnabled(
    execute=@n12.then
    self=Get GroupPatternDurationSlider()
    bInIsEnabled="Not Equal (Enum)"(A:EStageGroupOffset Enum=Get GroupOffset() -> (GroupOffset:EStageGroupOffset Enum=AllAtOnce), B:EStageGroupOffset Enum=AllAtOnce)
   )
   Branch(execute=@n12.LoopBody, Condition="Equal Exactly (String)"(A=Enum to String(Enumerator:EStageGroupOffset Enum=@n12), B=@n10.SelectedItem))
   > SetSelectedOption(self=Get RotationPatternCBox(), Option=@n10.SelectedItem)
   > Set GroupOffset(GroupOffset:EStageGroupOffset Enum=@n12) -> (Output_Get:EStageGroupOffset Enum=AllAtOnce)
   "On Pressed (ApplySettingButton)"() {is=title K2Node_ComponentBoundEvent}
   > Branch(Condition=Is Not Empty!(TargetArray:Array of BP Sparks Stage Show Actor Group Object References=Get StageShowActorGroups() -> (StageShowActorGroups:Array of BP Sparks Stage Show Actor Group Object References)) -> (ReturnValue:bool))
   > n13 = GetStageShowEventData!(execute, self) -> (then, EventSettings:Stage Show Event Settings Structure)
   n14 = "Get (a copy)"(
    Array:Array of BP Sparks Stage Show Actor Group Object References=Get StageShowActorGroups()
    Dimension 1=Get TargetSelectedIndex() -> (TargetSelectedIndex:int=0)
   ) -> (Output:BP Sparks Stage Show Actor Group Object Reference)
   > Is Valid(InputObject=@n14)
   > PlayEvent!(execute, self:BP Sparks Stage Show Actor Group Object Reference=@n14, Stage Show Event Settings:Stage Show Event Settings Structure=@n13) -> (then)
   n15 = "On Selection Changed (TargetCBox)"!() -> (OutputDelegate, then, SelectedItem:String, SelectionType:ESelectInfo Enum) {is=title K2Node_ComponentBoundEvent}
   > n16 = Switch on ESelectInfo(Selection=@n15.SelectionType)
   n17 = For Each Loop with Break!(
    Exec:Exec=[@n16.Direct, @n16.OnKeyPress, @n16.OnMouseClick]
    Array:Array of BP Sparks Stage Show Actor Group Object References=Get StageShowActorGroups()
    Break:Exec=@n19.then
   ) -> (LoopBody:Exec, Array Element:BP Sparks Stage Show Actor Group Object Reference, Array Index:int, Completed:Exec)
   n18 = Get Id!(self:BP Sparks Stage Show Actor Group Object Reference=@n17) -> (Id:Name=None)
   > Branch(Condition="Equal Exactly (String)"(A=@n15.SelectedItem, B="To String (Name)"(InName=@n18)))
   > n19 = Set TargetSelectedIndex(TargetSelectedIndex:int=@n17.Array Index) -> (Output_Get:int=0)
   "On Property Changed (PrimaryPaletteColor)"!() -> (OutputDelegate, then, PropertyName:Name=None) {is=title K2Node_ComponentBoundEvent}
   > SetVectorParameterValue!(
    execute
    WorldContextObject
    Collection:Material Parameter Collection Object Reference=/SparksStageShow/Systems/StageManager_v2_Prototype/MaterialFunctions/MPC_SparksStageShow
    ParameterName:Name=PrimaryColor
    ParameterValue:"Linear Color Structure (by ref)"=Get PrimaryColor() -> (PrimaryColor:Linear Color Structure)
   ) -> (then) {is=title K2Node_CallMaterialParameterCollectionFunction}
   GetAllActorGroupsInLevel() {is=title K2Node_CustomEvent}
   > ClearOptions(self:"Combo Box (String) Object Reference"=Get TargetCBox() -> (TargetCBox:Editor Utility Combo Box String Object Reference))
   > n20 = GetAllActorsOfClass!(
    execute
    WorldContextObject:Object Reference=Self-Reference()
    ActorClass:Actor Class Reference=/SparksStageShow/Systems/StageManager_v2_Prototype/Blueprints/BP_Sparks_StageShow_ActorGroup.BP_Sparks_StageShow_ActorGroup_C
   ) -> (then, OutActors:Array of BP Sparks Stage Show Actor Group Object References)
   > n21 = Set StageShowActorGroups(StageShowActorGroups:Array of BP Sparks Stage Show Actor Group Object References=@n20) -> (Output_Get:Array of BP Sparks Stage Show Actor Group Object References)
   > Branch(Condition=Is Not Empty(TargetArray=@n20))
   > n22 = For Each Loop(Array:Array of BP Sparks Stage Show Actor Group Object References=@n21) -> (Array Element:BP Sparks Stage Show Actor Group Object Reference)
   > AddOption!(
    execute
    self:"Combo Box (String) Object Reference"=Get TargetCBox()
    Option:String="To String (Name)"(InName=Get Id(self=@n22) -> (Id=None))
   ) -> (then)
   > InitializeActors(self:BP Sparks Stage Show Actor Group Object Reference=@n22)
   SetSelectedIndex!(execute=@n22.Completed, self:"Combo Box (String) Object Reference"=Get TargetCBox(), Index:int=0) -> (then)
   > Set TargetSelectedIndex(TargetSelectedIndex=0) -> (Output_Get=0)
   FillComboBoxValues() {is=title K2Node_CustomEvent}
   > n23 = Sequence!(execute) -> (then_0, then_1, then_2:Exec)
   > ClearOptions(self=Get RotationPlaybackCBox())
   > n24 = ForEach EStageShowRotationMode()
   > AddOption(self=Get RotationPlaybackCBox(), Option=Enum to String(Enumerator:EStageShowRotationMode Enum=@n24))
   > SetSelectedOption(
    self=Get RotationPlaybackCBox()
    Option=Enum to String(Enumerator=Literal enum EStageShowRotationMode!(Enum:EStageShowRotationMode Enum=PingPong) -> (ReturnValue:EStageShowRotationMode Enum) {is=title K2Node_EnumLiteral})
   )
   ClearOptions(execute=@n23.then_1, self=Get RotationPatternCBox())
   > n25 = ForEach EStageShowRotationPattern()
   > AddOption(self=Get RotationPatternCBox(), Option=Enum to String(Enumerator:EStageShowRotationPattern Enum=@n25))
   > SetSelectedOption(
    self=Get RotationPatternCBox()
    Option=Enum to String(Enumerator=Literal enum EStageShowRotationPattern!(Enum:EStageShowRotationPattern Enum=None) -> (ReturnValue:EStageShowRotationPattern Enum) {is=title K2Node_EnumLiteral})
   )
   ClearOptions(execute=@n23.then_2, self=Get GroupPatternCBox() -> (GroupPatternCBox:Editor Utility Combo Box String Object Reference))
   > n26 = ForEach EStageGroupOffset()
   > AddOption(self=Get GroupPatternCBox(), Option=Enum to String(Enumerator:EStageGroupOffset Enum=@n26))
   > SetSelectedOption(
    self=Get GroupPatternCBox()
    Option=Enum to String(Enumerator=Literal enum EStageGroupOffset!(Enum:EStageGroupOffset Enum=AllAtOnce) -> (ReturnValue:EStageGroupOffset Enum) {is=title K2Node_EnumLiteral})
   )
   n27 = OnMetronomeStateChange!() -> (OutputDelegate, then, IsPlaying:bool=~) {is=title K2Node_CustomEvent}
   > Branch(Condition=@n27.IsPlaying)
   Event PreConstruct()
   > Branch(Condition=NOT Boolean(A=Get isInitialized() -> (isInitialized:bool)))
   > FillComboBoxValues() {is=title K2Node_CallFunction}
   > GetAllActorGroupsInLevel() {is=title K2Node_CallFunction}
   > Bind Event to On State Change(
    self:Sparks Stage Show Utility Metronome Object Reference=Get MetronomeWidget() -> (MetronomeWidget:Sparks Stage Show Utility Metronome Object Reference)
    Delegate=Create Event()
   ) {is=title K2Node_AddDelegate}
   n28 = Get PrimaryPaletteColor() -> (PrimaryPaletteColor:Single Property View Object Reference)
   > SetObject(self=@n28, NewObject=Self-Reference())
   > SetPropertyName!(execute, self:Single Property View Object Reference=@n28, NewPropertyName:Name=PrimaryColor) -> (then)
   > Set isInitialized(isInitialized:bool=true) -> (Output_Get:bool)
   Event Construct()
  ]}
 GetStageShowEventData {nodes [
   GetStageShowEventData() {is=title K2Node_FunctionEntry}
   > n1 = GetSignalParameters!(
    execute
    self:Sparks Stage Show Utility Beat Signal Editor Object Reference=Get BeatSignalEditor_Intensity() -> (BeatSignalEditor_Intensity:Sparks Stage Show Utility Beat Signal Editor Object Reference)
   ) -> (then, Signal Data:Stage Show Signal Data Structure)
   > n2 = GetSignalParameters(self=Get BeatSignalEditor_Rotation())
   n3 = Get RotationPattern() -> (RotationPattern=None)
   n4 = Get RotationPlaybackType() -> (RotationPlaybackType:EStageShowRotationMode Enum=PingPong)
   n5 = Get GroupOffset() -> (GroupOffset=AllAtOnce)
   n6 = Get GroupOffsetDuration() -> (GroupOffsetDuration:double)
   > Return Node!(
    execute
    EventSettings:Stage Show Event Settings Structure=Make Stage Show Event Settings!(
     LightIntensitySignal:Stage Show Signal Data Structure=@n1
     RotationSignal:Stage Show Signal Data Structure=@n2
     RotationSettings:Stage Show Rotation Settings Structure=Make Stage Show Rotation Settings!(Pattern:EStageShowRotationPattern Enum=@n3, AnimationPlayback:EStageShowRotationMode Enum=@n4) -> (StageShowRotationSettings:Stage Show Rotation Settings Structure) {is=title K2Node_MakeStruct}
     GroupPatternSettings:Stage Show Group Pattern Data Structure=Make Stage Show Group Pattern Data!(OffsetType:EStageGroupOffset Enum=@n5, Duration:double=@n6) -> (StageShowGroupPatternData:Stage Show Group Pattern Data Structure) {is=title K2Node_MakeStruct}
    ) -> (StageShowEventSettings:Stage Show Event Settings Structure) {is=title K2Node_MakeStruct}
   )
  ]}
]
variables [
 BeatAudio {type=AudioComponent, default=null}
 GroupOffset {type=EStageGroupOffset, default=EStageGroupOffset_MAX}
 GroupOffsetDuration {type=number, default=1}
 isInitialized {type=boolean}
 MeasureAudio {type=AudioComponent, default=null}
 PrimaryColor {type=LinearColor, default=<rgba 1 1 1 1>}
 RotationPattern {type=EStageShowRotationPattern, default=EStageShowRotationPattern_MAX}
 RotationPlaybackType {type=EStageShowRotationMode, default=EStageShowRotationMode_MAX}
 StageShowActorGroups {type="array<BP_Sparks_StageShow_ActorGroup_C>", default []}
 TargetSelectedIndex {type=integer}
]
widgets {
 info {rootWidgetClass=&HorizontalBox, widgetCount=48}
 widgets [
  HorizontalBox_0 {widgetClassPath=&HorizontalBox, widget=HorizontalBox_0}
  SizeBox_2 {
   widgetClassPath=&SizeBox
   parent=HorizontalBox_0
   widget=SizeBox_2
   properties {bOverride_MinDesiredWidth=true, minDesiredWidth=310}
   slot {class=HorizontalBoxSlot, properties {padding=<ltrb 0 20 0 0>, size {sizeRule=Fill, value=1}}}
  }
  VerticalBox_1 {
   widgetClassPath=&VerticalBox
   parent=SizeBox_2
   widget=VerticalBox_1
   slot {class=SizeBoxSlot, properties {verticalAlignment=VAlign_Top, padding=<ltrb 20 0 15 0>}}
  }
  HorizontalBox_5 {widgetClassPath=&HorizontalBox, parent=VerticalBox_1, widget=HorizontalBox_5, slot {class=VerticalBoxSlot}}
  VerticalBox_2 {
   widgetClassPath=&VerticalBox
   parent=HorizontalBox_5
   widget=VerticalBox_2
   properties {bOverride_Cursor=true}
   slot {class=HorizontalBoxSlot, properties {padding=<ltrb 0 0 10 5>, size {sizeRule=Fill, value=1}}}
  }
  TextBlock_5 {
   widgetClassPath=&TextBlock
   parent=VerticalBox_2
   widget=TextBlock_5
   properties {
    text=Target
    font {
     bForceMonospaced=false
     bMaterialIsStencil=false
     fontMaterial=null
     fontObject=&Roboto
     letterSpacing=0
     monospacedWidth=1
     size=13.5
     skewAmount=0
     typefaceFontName=Bold
     outlineSettings {
      bApplyOutlineToDropShadows=false
      bMiteredCorners=false
      bSeparateFillAlpha=false
      outlineMaterial=null
      outlineSize=0
      outlineColor=<rgba 0 0 0 1>
     }
    }
   }
   slot {class=VerticalBoxSlot, properties {padding=<ltrb 0 0 0 5>}}
  }
  SizeBox_7 {
   widgetClassPath=&SizeBox
   parent=VerticalBox_2
   widget=SizeBox_7
   properties {bOverride_MinDesiredHeight=true, bOverride_MinDesiredWidth=true, minDesiredHeight=32, minDesiredWidth=160}
   slot {class=VerticalBoxSlot}
  }
  TargetCBox {
   widgetClassPath=/Script/Blutility.EditorUtilityComboBoxString
   bIsVariable=true
   parent=SizeBox_7
   widget=TargetCBox
   properties {
    selectedOption=Bounce Up/Down
    defaultOptions [None, Bounce Up/Down, Circle]
    font {
     bForceMonospaced=false
     bMaterialIsStencil=false
     fontMaterial=null
     fontObject=&Roboto
     letterSpacing=0
     monospacedWidth=1
     size=10.5
     skewAmount=0
     typefaceFontName=Regular
     outlineSettings {
      bApplyOutlineToDropShadows=false
      bMiteredCorners=false
      bSeparateFillAlpha=false
      outlineMaterial=null
      outlineSize=0
      outlineColor=<rgba 0 0 0 1>
     }
    }
    foregroundColor {colorUseRule=UseColor_Specified, specifiedColor=<rgba 0.5271 0.5271 0.5271 1>}
   }
   slot {class=SizeBoxSlot}
  }
  SizeBox_0 {
   widgetClassPath=&SizeBox
   parent=HorizontalBox_5
   widget=SizeBox_0
   properties {bOverride_MinDesiredWidth=true, minDesiredWidth=350}
   slot {class=HorizontalBoxSlot, properties {horizontalAlignment=HAlign_Right}}
  }
  MetronomeWidget {
   widgetClassPath=&/Sparks_StageShowUtility_Metronome
   bIsVariable=true
   parent=SizeBox_0
   widget=MetronomeWidget
   slot {class=SizeBoxSlot, properties {verticalAlignment=VAlign_Bottom}}
  }
  Divider {
   widgetClassPath=&Image
   parent=VerticalBox_1
   widget=Divider
   properties {
    brush {
     bIsDynamicallyLoaded=false
     drawAs=Image
     imageType=NoImage
     mirroring=NoMirror
     resourceName=None
     resourceObject=null
     tiling=NoTile
     imageSize=<32 1>
     margin=<ltrb 0 0 0 0>
     outlineSettings {
      bUseBrushTransparency=false
      roundingType=HalfHeightRadius
      width=0
      color {colorUseRule=UseColor_Specified, specifiedColor=<rgba 0 0 0 0>}
      cornerRadii=<0 0 0 0>
     }
     tintColor {colorUseRule=UseColor_Specified, specifiedColor=<rgba 1 1 1 1>}
     uVRegion {bIsValid=false, max=<0 0>, min=<0 0>}
    }
    colorAndOpacity=<rgba 1 1 1 0.25>
   }
   slot {class=VerticalBoxSlot, properties {padding=<ltrb 0 5 0 0>}}
  }
  HorizontalBox_3 {widgetClassPath=&HorizontalBox, parent=VerticalBox_1, widget=HorizontalBox_3, slot {class=VerticalBoxSlot}}
  VerticalBox_138 {
   widgetClassPath=&VerticalBox
   parent=HorizontalBox_3
   widget=VerticalBox_138
   slot {class=HorizontalBoxSlot, properties {padding=<ltrb 0 0 10 0>}}
  }
  TextBlock_67 {
   widgetClassPath=&TextBlock
   parent=VerticalBox_138
   widget=TextBlock_67
   properties {
    text=Rotation Settings
    toolTipText=Settings to select the rotation animation.
    font {
     bForceMonospaced=false
     bMaterialIsStencil=false
     fontMaterial=null
     fontObject=&Roboto
     letterSpacing=0
     monospacedWidth=1
     size=10.5
     skewAmount=0
     typefaceFontName=Bold
     outlineSettings {
      bApplyOutlineToDropShadows=false
      bMiteredCorners=false
      bSeparateFillAlpha=false
      outlineMaterial=null
      outlineSize=0
      outlineColor=<rgba 0 0 0 1>
     }
    }
   }
   slot {class=VerticalBoxSlot, properties {padding=<ltrb 0 10 0 10>}}
  }
  HorizontalBox_36 {
   widgetClassPath=&HorizontalBox
   parent=VerticalBox_138
   widget=HorizontalBox_36
   slot {class=VerticalBoxSlot, properties {padding=<ltrb 0 0 0 5>}}
  }
  TextBlock_32 {
   widgetClassPath=&TextBlock
   parent=HorizontalBox_36
   widget=TextBlock_32
   properties {
    text=Pattern
    toolTipText=Which rotation pattern will be used.
    font {
     bForceMonospaced=false
     bMaterialIsStencil=false
     fontMaterial=null
     fontObject=&Roboto
     letterSpacing=0
     monospacedWidth=1
     size=9
     skewAmount=0
     typefaceFontName=Regular
     outlineSettings {
      bApplyOutlineToDropShadows=false
      bMiteredCorners=false
      bSeparateFillAlpha=false
      outlineMaterial=null
      outlineSize=0
      outlineColor=<rgba 0 0 0 1>
     }
    }
   }
   slot {class=HorizontalBoxSlot, properties {verticalAlignment=VAlign_Center}}
  }
  SizeBox_1 {
   widgetClassPath=&SizeBox
   parent=HorizontalBox_36
   widget=SizeBox_1
   properties {bOverride_MinDesiredWidth=true, minDesiredWidth=160}
   slot {class=HorizontalBoxSlot, properties {horizontalAlignment=HAlign_Right, verticalAlignment=VAlign_Center, size {sizeRule=Fill, value=1}}}
  }
  RotationPatternCBox {
   widgetClassPath=&EditorUtilityComboBoxString
   bIsVariable=true
   parent=SizeBox_1
   widget=RotationPatternCBox
   properties {
    selectedOption=Bounce Up/Down
    defaultOptions [None, Bounce Up/Down, Circle]
    font {
     bForceMonospaced=false
     bMaterialIsStencil=false
     fontMaterial=null
     fontObject=&Roboto
     letterSpacing=0
     monospacedWidth=1
     size=9
     skewAmount=0
     typefaceFontName=Regular
     outlineSettings {
      bApplyOutlineToDropShadows=false
      bMiteredCorners=false
      bSeparateFillAlpha=false
      outlineMaterial=null
      outlineSize=0
      outlineColor=<rgba 0 0 0 1>
     }
    }
    foregroundColor {colorUseRule=UseColor_Specified, specifiedColor=<rgba 0.5271 0.5271 0.5271 1>}
   }
   slot {class=SizeBoxSlot}
  }
  HorizontalBox {widgetClassPath=&HorizontalBox, parent=VerticalBox_138, widget=HorizontalBox, slot {class=VerticalBoxSlot}}
  TextBlock {
   widgetClassPath=&TextBlock
   parent=HorizontalBox
   widget=TextBlock
   properties {
    text=Playback Type
    toolTipText="How to animation playback will behave.\r\n\r\nPing-Pong: animation will go from 0 to 1, then 1 to 0.\r\n\r\nLinear: animation wil go 0 to 1. Useful for animations that looping is authored on the animation itself."
    font {
     bForceMonospaced=false
     bMaterialIsStencil=false
     fontMaterial=null
     fontObject=&Roboto
     letterSpacing=0
     monospacedWidth=1
     size=9
     skewAmount=0
     typefaceFontName=Regular
     outlineSettings {
      bApplyOutlineToDropShadows=false
      bMiteredCorners=false
      bSeparateFillAlpha=false
      outlineMaterial=null
      outlineSize=0
      outlineColor=<rgba 0 0 0 1>
     }
    }
   }
   slot {class=HorizontalBoxSlot, properties {verticalAlignment=VAlign_Center}}
  }
  SizeBox {
   widgetClassPath=&SizeBox
   parent=HorizontalBox
   widget=SizeBox
   properties {bOverride_MinDesiredWidth=true, minDesiredWidth=160}
   slot {class=HorizontalBoxSlot, properties {horizontalAlignment=HAlign_Right, verticalAlignment=VAlign_Center, size {sizeRule=Fill, value=1}}}
  }
  RotationPlaybackCBox {
   widgetClassPath=&EditorUtilityComboBoxString
   bIsVariable=true
   parent=SizeBox
   widget=RotationPlaybackCBox
   properties {
    selectedOption=Bounce Up/Down
    defaultOptions [None, Bounce Up/Down, Circle]
    font {
     bForceMonospaced=false
     bMaterialIsStencil=false
     fontMaterial=null
     fontObject=&Roboto
     letterSpacing=0
     monospacedWidth=1
     size=9
     skewAmount=0
     typefaceFontName=Regular
     outlineSettings {
      bApplyOutlineToDropShadows=false
      bMiteredCorners=false
      bSeparateFillAlpha=false
      outlineMaterial=null
      outlineSize=0
      outlineColor=<rgba 0 0 0 1>
     }
    }
    foregroundColor {colorUseRule=UseColor_Specified, specifiedColor=<rgba 0.5271 0.5271 0.5271 1>}
   }
   slot {class=SizeBoxSlot}
  }
  TextBlock_1 {
   widgetClassPath=&TextBlock
   parent=VerticalBox_138
   widget=TextBlock_1
   properties {
    text=Group Pattern
    toolTipText="Setting for how  the signal will propagate through the group members"
    font {
     bForceMonospaced=false
     bMaterialIsStencil=false
     fontMaterial=null
     fontObject=&Roboto
     letterSpacing=0
     monospacedWidth=1
     size=10.5
     skewAmount=0
     typefaceFontName=Bold
     outlineSettings {
      bApplyOutlineToDropShadows=false
      bMiteredCorners=false
      bSeparateFillAlpha=false
      outlineMaterial=null
      outlineSize=0
      outlineColor=<rgba 0 0 0 1>
     }
    }
   }
   slot {class=VerticalBoxSlot, properties {padding=<ltrb 0 20 0 10>}}
  }
  HorizontalBox_1 {
   widgetClassPath=&HorizontalBox
   parent=VerticalBox_138
   widget=HorizontalBox_1
   properties {toolTipText=How the signal will spread through the actors.}
   slot {class=VerticalBoxSlot, properties {padding=<ltrb 0 0 0 5>}}
  }
  TextBlock_2 {
   widgetClassPath=&TextBlock
   parent=HorizontalBox_1
   widget=TextBlock_2
   properties {
    text=Pattern
    font {
     bForceMonospaced=false
     bMaterialIsStencil=false
     fontMaterial=null
     fontObject=&Roboto
     letterSpacing=0
     monospacedWidth=1
     size=9
     skewAmount=0
     typefaceFontName=Regular
     outlineSettings {
      bApplyOutlineToDropShadows=false
      bMiteredCorners=false
      bSeparateFillAlpha=false
      outlineMaterial=null
      outlineSize=0
      outlineColor=<rgba 0 0 0 1>
     }
    }
   }
   slot {class=HorizontalBoxSlot, properties {verticalAlignment=VAlign_Center}}
  }
  SizeBox_3 {
   widgetClassPath=&SizeBox
   parent=HorizontalBox_1
   widget=SizeBox_3
   properties {bOverride_MinDesiredWidth=true, minDesiredWidth=160}
   slot {class=HorizontalBoxSlot, properties {horizontalAlignment=HAlign_Right, verticalAlignment=VAlign_Center, size {sizeRule=Fill, value=1}}}
  }
  GroupPatternCBox {
   widgetClassPath=&EditorUtilityComboBoxString
   bIsVariable=true
   parent=SizeBox_3
   widget=GroupPatternCBox
   properties {
    selectedOption=Bounce Up/Down
    defaultOptions [None, Bounce Up/Down, Circle]
    font {
     bForceMonospaced=false
     bMaterialIsStencil=false
     fontMaterial=null
     fontObject=&Roboto
     letterSpacing=0
     monospacedWidth=1
     size=9
     skewAmount=0
     typefaceFontName=Regular
     outlineSettings {
      bApplyOutlineToDropShadows=false
      bMiteredCorners=false
      bSeparateFillAlpha=false
      outlineMaterial=null
      outlineSize=0
      outlineColor=<rgba 0 0 0 1>
     }
    }
    foregroundColor {colorUseRule=UseColor_Specified, specifiedColor=<rgba 0.5271 0.5271 0.5271 1>}
   }
   slot {class=SizeBoxSlot}
  }
  HorizontalBox_2 {
   widgetClassPath=&HorizontalBox
   parent=VerticalBox_138
   widget=HorizontalBox_2
   slot {class=VerticalBoxSlot, properties {padding=<ltrb 0 0 0 5>}}
  }
  TextBlock_3 {
   widgetClassPath=&TextBlock
   parent=HorizontalBox_2
   widget=TextBlock_3
   properties {
    text=Duration
    toolTipText="How long the signal will take to reach the last actor on the group.\r\nUnit of duration is measures."
    font {
     bForceMonospaced=false
     bMaterialIsStencil=false
     fontMaterial=null
     fontObject=&Roboto
     letterSpacing=0
     monospacedWidth=1
     size=9
     skewAmount=0
     typefaceFontName=Regular
     outlineSettings {
      bApplyOutlineToDropShadows=false
      bMiteredCorners=false
      bSeparateFillAlpha=false
      outlineMaterial=null
      outlineSize=0
      outlineColor=<rgba 0 0 0 1>
     }
    }
   }
   slot {class=HorizontalBoxSlot, properties {verticalAlignment=VAlign_Center}}
  }
  SizeBox_4 {
   widgetClassPath=&SizeBox
   parent=HorizontalBox_2
   widget=SizeBox_4
   properties {bOverride_MinDesiredWidth=true, minDesiredWidth=160}
   slot {class=HorizontalBoxSlot, properties {horizontalAlignment=HAlign_Right, verticalAlignment=VAlign_Center, size {sizeRule=Fill, value=1}}}
  }
  GroupPatternDurationSlider {
   widgetClassPath=/Script/Blutility.EditorUtilitySlider
   bIsVariable=true
   parent=SizeBox_4
   widget=GroupPatternDurationSlider
   properties {
    minValue=0.25
    toolTipText=1
    value=1
    widgetStyle {
     barThickness=8
     disabledBarImage {
      bIsDynamicallyLoaded=false
      drawAs=RoundedBox
      imageType=NoImage
      mirroring=NoMirror
      resourceName=None
      resourceObject=null
      tiling=NoTile
      imageSize=<0 0>
      margin=<ltrb 0 0 0 0>
      outlineSettings {
       bUseBrushTransparency=false
       roundingType=FixedRadius
       width=1
       color {colorUseRule=UseColor_ColorTable, specifiedColor=<rgba 0 0 0 0>}
       cornerRadii=<2 2 2 2>
      }
      tintColor {colorUseRule=UseColor_Specified, specifiedColor=<rgba 0.0048 0.0048 0.0048 1>}
      uVRegion {bIsValid=false, max=<0 0>, min=<0 0>}
     }
     disabledThumbImage {
      bIsDynamicallyLoaded=false
      drawAs=RoundedBox
      imageType=NoImage
      mirroring=NoMirror
      resourceName=None
      resourceObject=null
      tiling=NoTile
      imageSize=<12 12>
      margin=<ltrb 0 0 0 0>
      outlineSettings {
       bUseBrushTransparency=false
       roundingType=HalfHeightRadius
       width=0
       color {colorUseRule=UseColor_Specified, specifiedColor=<rgba 0 0 0 0>}
       cornerRadii=<0 0 0 0>
      }
      tintColor {colorUseRule=UseColor_Specified, specifiedColor=<rgba 0.5271 0.5271 0.5271 1>}
      uVRegion {bIsValid=false, max=<0 0>, min=<0 0>}
     }
     hoveredBarImage {
      bIsDynamicallyLoaded=false
      drawAs=RoundedBox
      imageType=NoImage
      mirroring=NoMirror
      resourceName=None
      resourceObject=null
      tiling=NoTile
      imageSize=<0 0>
      margin=<ltrb 0 0 0 0>
      outlineSettings {
       bUseBrushTransparency=false
       roundingType=FixedRadius
       width=1
       color {colorUseRule=UseColor_ColorTable, specifiedColor=<rgba 0 0 0 0>}
       cornerRadii=<2 2 2 2>
      }
      tintColor {colorUseRule=UseColor_Specified, specifiedColor=<rgba 0.0048 0.0048 0.0048 1>}
      uVRegion {bIsValid=false, max=<0 0>, min=<0 0>}
     }
     hoveredThumbImage {
      bIsDynamicallyLoaded=false
      drawAs=RoundedBox
      imageType=NoImage
      mirroring=NoMirror
      resourceName=None
      resourceObject=null
      tiling=NoTile
      imageSize=<12 12>
      margin=<ltrb 0 0 0 0>
      outlineSettings {
       bUseBrushTransparency=false
       roundingType=HalfHeightRadius
       width=0
       color {colorUseRule=UseColor_Specified, specifiedColor=<rgba 0 0 0 0>}
       cornerRadii=<0 0 0 0>
      }
      tintColor {colorUseRule=UseColor_Specified, specifiedColor=<rgba 1 1 1 1>}
      uVRegion {bIsValid=false, max=<0 0>, min=<0 0>}
     }
     normalBarImage {
      bIsDynamicallyLoaded=false
      drawAs=RoundedBox
      imageType=NoImage
      mirroring=NoMirror
      resourceName=None
      resourceObject=null
      tiling=NoTile
      imageSize=<0 0>
      margin=<ltrb 0 0 0 0>
      outlineSettings {
       bUseBrushTransparency=false
       roundingType=FixedRadius
       width=1
       color {colorUseRule=UseColor_ColorTable, specifiedColor=<rgba 0 0 0 0>}
       cornerRadii=<2 2 2 2>
      }
      tintColor {colorUseRule=UseColor_Specified, specifiedColor=<rgba 0.0048 0.0048 0.0048 1>}
      uVRegion {bIsValid=false, max=<0 0>, min=<0 0>}
     }
     normalThumbImage {
      bIsDynamicallyLoaded=false
      drawAs=RoundedBox
      imageType=NoImage
      mirroring=NoMirror
      resourceName=None
      resourceObject=null
      tiling=NoTile
      imageSize=<12 12>
      margin=<ltrb 0 0 0 0>
      outlineSettings {
       bUseBrushTransparency=false
       roundingType=HalfHeightRadius
       width=0
       color {colorUseRule=UseColor_Specified, specifiedColor=<rgba 0 0 0 0>}
       cornerRadii=<0 0 0 0>
      }
      tintColor {colorUseRule=UseColor_Specified, specifiedColor=<rgba 0.5271 0.5271 0.5271 1>}
      uVRegion {bIsValid=false, max=<0 0>, min=<0 0>}
     }
    }
   }
   slot {class=SizeBoxSlot}
  }
  TextBlock_6 {
   widgetClassPath=&TextBlock
   parent=VerticalBox_138
   widget=TextBlock_6
   properties {
    text=Color Palette Preview
    toolTipText="Setting for how  the signal will propagate through the group members"
    font {
     bForceMonospaced=false
     bMaterialIsStencil=false
     fontMaterial=null
     fontObject=&Roboto
     letterSpacing=0
     monospacedWidth=1
     size=10.5
     skewAmount=0
     typefaceFontName=Bold
     outlineSettings {
      bApplyOutlineToDropShadows=false
      bMiteredCorners=false
      bSeparateFillAlpha=false
      outlineMaterial=null
      outlineSize=0
      outlineColor=<rgba 0 0 0 1>
     }
    }
   }
   slot {class=VerticalBoxSlot, properties {padding=<ltrb 0 20 0 10>}}
  }
  PrimaryPaletteColor {
   widgetClassPath=&SinglePropertyView
   parent=VerticalBox_138
   widget=PrimaryPaletteColor
   properties {propertyName=PrimaryPaletteColor}
   slot {class=VerticalBoxSlot}
  }
  ApplySettingButton {
   widgetClassPath=&EditorUtilityButton
   parent=VerticalBox_138
   widget=ApplySettingButton
   properties {
    toolTipText=Applies current signals to the targeted actor group.
    widgetStyle {
     clickedSlateSound {resourceObject=null}
     disabled {
      bIsDynamicallyLoaded=false
      drawAs=RoundedBox
      imageType=NoImage
      mirroring=NoMirror
      resourceName=None
      resourceObject=null
      tiling=NoTile
      imageSize=<0 0>
      margin=<ltrb 0 0 0 0>
      outlineSettings {
       bUseBrushTransparency=false
       roundingType=FixedRadius
       width=1
       color {colorUseRule=UseColor_ColorTable, specifiedColor=<rgba 0 0 0 0>}
       cornerRadii=<4 4 4 4>
      }
      tintColor {colorUseRule=UseColor_Specified, specifiedColor=<rgba 0.0395 0.0395 0.0395 1>}
      uVRegion {bIsValid=false, max=<0 0>, min=<0 0>}
     }
     disabledForeground {colorUseRule=UseColor_ColorTable, specifiedColor=<rgba 0 0 0 0>}
     hovered {
      bIsDynamicallyLoaded=false
      drawAs=RoundedBox
      imageType=NoImage
      mirroring=NoMirror
      resourceName=None
      resourceObject=null
      tiling=NoTile
      imageSize=<32 32>
      margin=<ltrb 0 0 0 0>
      outlineSettings {
       bUseBrushTransparency=false
       roundingType=FixedRadius
       width=1
       color {colorUseRule=UseColor_ColorTable, specifiedColor=<rgba 0 0 0 0>}
       cornerRadii=<4 4 4 4>
      }
      tintColor {colorUseRule=UseColor_Specified, specifiedColor=<rgba 0 0.1979 0.0497 1>}
      uVRegion {bIsValid=false, max=<0 0>, min=<0 0>}
     }
     hoveredForeground {colorUseRule=UseColor_ColorTable, specifiedColor=<rgba 0 0 0 0>}
     hoveredSlateSound {resourceObject=null}
     normal {
      bIsDynamicallyLoaded=false
      drawAs=RoundedBox
      imageType=NoImage
      mirroring=NoMirror
      resourceName=None
      resourceObject=null
      tiling=NoTile
      imageSize=<32 32>
      margin=<ltrb 0 0 0 0>
      outlineSettings {
       bUseBrushTransparency=false
       roundingType=FixedRadius
       width=1
       color {colorUseRule=UseColor_ColorTable, specifiedColor=<rgba 0 0 0 0>}
       cornerRadii=<4 4 4 4>
      }
      tintColor {colorUseRule=UseColor_Specified, specifiedColor=<rgba 0 0.0393 0.0395 1>}
      uVRegion {bIsValid=false, max=<0 0>, min=<0 0>}
     }
     normalForeground {colorUseRule=UseColor_ColorTable, specifiedColor=<rgba 0 0 0 0>}
     normalPadding=<ltrb 12 1.5 12 1.5>
     pressed {
      bIsDynamicallyLoaded=false
      drawAs=RoundedBox
      imageType=NoImage
      mirroring=NoMirror
      resourceName=None
      resourceObject=null
      tiling=NoTile
      imageSize=<32 32>
      margin=<ltrb 0 0 0 0>
      outlineSettings {
       bUseBrushTransparency=false
       roundingType=FixedRadius
       width=1
       color {colorUseRule=UseColor_ColorTable, specifiedColor=<rgba 0 0 0 0>}
       cornerRadii=<4 4 4 4>
      }
      tintColor {colorUseRule=UseColor_Specified, specifiedColor=<rgba 0.0037 0.0156 0.0104 1>}
      uVRegion {bIsValid=false, max=<0 0>, min=<0 0>}
     }
     pressedForeground {colorUseRule=UseColor_ColorTable, specifiedColor=<rgba 0 0 0 0>}
     pressedPadding=<ltrb 12 2.5 12 0.5>
     pressedSlateSound {resourceObject=null}
    }
   }
   slot {class=VerticalBoxSlot, properties {padding=<ltrb 0 10 0 0>}}
  }
  TextBlock_116 {
   widgetClassPath=&TextBlock
   parent=ApplySettingButton
   widget=TextBlock_116
   properties {
    text=Apply Settings
    font {
     bForceMonospaced=false
     bMaterialIsStencil=false
     fontMaterial=null
     fontObject=&Roboto
     letterSpacing=0
     monospacedWidth=1
     size=10.5
     skewAmount=0
     typefaceFontName=Bold
     outlineSettings {
      bApplyOutlineToDropShadows=false
      bMiteredCorners=false
      bSeparateFillAlpha=false
      outlineMaterial=null
      outlineSize=0
      outlineColor=<rgba 0 0 0 1>
     }
    }
   }
   slot {class=ButtonSlot, properties {padding=<ltrb 4 4 4 4>}}
  }
  HorizontalBox_4 {
   widgetClassPath=&HorizontalBox
   parent=HorizontalBox_3
   widget=HorizontalBox_4
   slot {class=HorizontalBoxSlot, properties {size {sizeRule=Fill, value=1}}}
  }
  Image_294 {
   widgetClassPath=&Image
   parent=HorizontalBox_4
   widget=Image_294
   properties {
    brush {
     bIsDynamicallyLoaded=false
     drawAs=Image
     imageType=NoImage
     mirroring=NoMirror
     resourceName=None
     resourceObject=null
     tiling=NoTile
     imageSize=<1 32>
     margin=<ltrb 0 0 0 0>
     outlineSettings {
      bUseBrushTransparency=false
      roundingType=HalfHeightRadius
      width=0
      color {colorUseRule=UseColor_Specified, specifiedColor=<rgba 0 0 0 0>}
      cornerRadii=<0 0 0 0>
     }
     tintColor {colorUseRule=UseColor_Specified, specifiedColor=<rgba 1 1 1 1>}
     uVRegion {bIsValid=false, max=<0 0>, min=<0 0>}
    }
    colorAndOpacity=<rgba 1 1 1 0.25>
   }
   slot {class=HorizontalBoxSlot}
  }
  VerticalBox_0 {
   widgetClassPath=&VerticalBox
   parent=HorizontalBox_4
   widget=VerticalBox_0
   slot {class=HorizontalBoxSlot, properties {size {sizeRule=Fill, value=1}}}
  }
  SizeBox_5 {
   widgetClassPath=&SizeBox
   parent=VerticalBox_0
   widget=SizeBox_5
   properties {bOverride_MinDesiredHeight=true, minDesiredHeight=30}
   slot {class=VerticalBoxSlot, properties {padding=<ltrb 0 0 0 10>}}
  }
  Overlay_52 {widgetClassPath=&Overlay, parent=SizeBox_5, widget=Overlay_52, slot {class=SizeBoxSlot}}
  Image_164 {
   widgetClassPath=&Image
   parent=Overlay_52
   widget=Image_164
   properties {colorAndOpacity=<rgba 0.0417 0.0417 0.0417 1>}
   slot {class=OverlaySlot, properties {horizontalAlignment=HAlign_Fill, verticalAlignment=VAlign_Fill}}
  }
  TextBlock_105 {
   widgetClassPath=&TextBlock
   parent=Overlay_52
   widget=TextBlock_105
   properties {
    text=Intensity Signal
    toolTipText=Signal that drives the luminosity of the actors.
    font {
     bForceMonospaced=false
     bMaterialIsStencil=false
     fontMaterial=null
     fontObject=&Roboto
     letterSpacing=0
     monospacedWidth=1
     size=10.5
     skewAmount=0
     typefaceFontName=Bold
     outlineSettings {
      bApplyOutlineToDropShadows=false
      bMiteredCorners=false
      bSeparateFillAlpha=false
      outlineMaterial=null
      outlineSize=0
      outlineColor=<rgba 0 0 0 1>
     }
    }
   }
   slot {class=OverlaySlot, properties {verticalAlignment=VAlign_Bottom, padding=<ltrb 10 0 0 5>}}
  }
  BeatSignalEditor_Intensity {
   widgetClassPath=&/Sparks_StageShowUtility_BeatSignalEditor
   bIsVariable=true
   parent=VerticalBox_0
   widget=BeatSignalEditor_Intensity
   slot {class=VerticalBoxSlot, properties {padding=<ltrb 0 0 0 10>, size {sizeRule=Fill, value=1}}}
  }
  SizeBox_6 {
   widgetClassPath=&SizeBox
   parent=VerticalBox_0
   widget=SizeBox_6
   properties {bOverride_MinDesiredHeight=true, minDesiredHeight=30}
   slot {class=VerticalBoxSlot, properties {padding=<ltrb 0 0 0 10>}}
  }
  Overlay {widgetClassPath=&Overlay, parent=SizeBox_6, widget=Overlay, slot {class=SizeBoxSlot}}
  Image {
   widgetClassPath=&Image
   parent=Overlay
   widget=Image
   properties {colorAndOpacity=<rgba 0.0423 0.0423 0.0423 1>}
   slot {class=OverlaySlot, properties {horizontalAlignment=HAlign_Fill, verticalAlignment=VAlign_Fill}}
  }
  TextBlock_4 {
   widgetClassPath=&TextBlock
   parent=Overlay
   widget=TextBlock_4
   properties {
    text=Rotation Signal
    toolTipText=Signal that drives the progression of the rotation animations.
    font {
     bForceMonospaced=false
     bMaterialIsStencil=false
     fontMaterial=null
     fontObject=&Roboto
     letterSpacing=0
     monospacedWidth=1
     size=10.5
     skewAmount=0
     typefaceFontName=Bold
     outlineSettings {
      bApplyOutlineToDropShadows=false
      bMiteredCorners=false
      bSeparateFillAlpha=false
      outlineMaterial=null
      outlineSize=0
      outlineColor=<rgba 0 0 0 1>
     }
    }
   }
   slot {class=OverlaySlot, properties {verticalAlignment=VAlign_Bottom, padding=<ltrb 10 0 0 5>}}
  }
  BeatSignalEditor_Rotation {
   widgetClassPath=&Sparks_StageShowUtility_BeatSignalEditor
   bIsVariable=true
   parent=VerticalBox_0
   widget=BeatSignalEditor_Rotation
   slot {class=VerticalBoxSlot, properties {size {sizeRule=Fill, value=1}}}
  }
 ]
}
